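# Control Rig for the F_MED_PolkaSkate female body cosmetic, procedurally posing its skirt, laces and buttons atop a shared DeformRig function.
# Twist-removed thigh helpers and weighted parent constraints carry the skirt and accessory bones.
# Spherical pose readers on thighs and calves weight corrective skirt transforms, some interpolated.
# State variables drive tombstone-emote and slide poses, plus Kalman-smoothed lowering of the skirt's ground collision when floor slope is disabled.
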ControlRigBlueprint /BRCosmetics/Characters/Player/Female/Medium/Bodies/F_MED_PolkaSkate/Meshes/F_MED_PolkaSkate_CtrlRig
graphs [
 RigVMFunctionLibrary {links [], nodes []}
 RigVMFunction_Sequence_1_ContainedGraph {nodes [
   n1 = RigVMFunctionEntryNode()
   RigVMFunction_Sequence = &RigVMFunction_Sequence(ExecuteContext=@n1.ExecuteContext)
   n2 = &RigVMFunction_Sequence(ExecuteContext=@RigVMFunction_Sequence.B)
   n3 = &RigVMFunction_Sequence(ExecuteContext=@n2.B)
   n4 = &RigVMFunction_Sequence(ExecuteContext=@n3.B)
   n5 = &RigVMFunction_Sequence(ExecuteContext=@n4.B)
   n6 = &RigVMFunction_Sequence(ExecuteContext=@n5.B)
   n7 = &RigVMFunction_Sequence(ExecuteContext=@n6.B)
   n8 = &RigVMFunction_Sequence(ExecuteContext=@n7.B)
   n9 = &RigVMFunction_Sequence(ExecuteContext=@n8.B)
   RigVMFunctionReturnNode(A=@RigVMFunction_Sequence.A, B=@n2.A, C=@n3.A, D=@n4.A, E=@n5.A, F=@n6.A, G=@n7.A, H=@n8.A, I=@n9.A, J=@n9.B)
  ]}
 RigVMFunction_Sequence_2_ContainedGraph {nodes [
   n1 = RigVMFunctionEntryNode()
   n2 = &RigVMFunction_Sequence(ExecuteContext=@n1.ExecuteContext)
   n3 = &RigVMFunction_Sequence(ExecuteContext=@n2.B)
   RigVMFunctionReturnNode(A=@n2.A, B=@n3.A, C=@n3.B)
  ]}
 RigVMModel {nodes [
   &RigUnit_ParentConstraint(
    AdvancedSettings=(InterpolationType=Average,RotationOrderForFilter=XZY)
    Child=(Type=Bone,Name="dyn_skirt_front_mid_MSTR")
    ExecutePin=()
    Parents=((Item=(Type=Bone,Name="thigh_l"),Weight=1),(Item=(Type=Bone,Name="thigh_r"),Weight=1))
    Weight=1
   )
   > &RigUnit_ParentConstraint(
    AdvancedSettings=(InterpolationType=Average,RotationOrderForFilter=XZY)
    Child=(Type=Bone,Name="dyn_skirt_front_l_MSTR")
    Parents=((Item=(Type=Bone,Name="thigh_l"),Weight=1))
    Weight=1
   )
   > &RigUnit_ParentConstraint(
    AdvancedSettings=(InterpolationType=Average,RotationOrderForFilter=XZY)
    Child=(Type=Bone,Name="dyn_skirt_side_l_MSTR")
    Parents=((Item=(Type=Bone,Name="thigh_l"),Weight=1))
    Weight=1
   )
   > &RigUnit_ParentConstraint(
    AdvancedSettings=(InterpolationType=Average,RotationOrderForFilter=XZY)
    Child=(Type=Bone,Name="dyn_skirt_back_l_MSTR")
    Parents=((Item=(Type=Bone,Name="thigh_l"),Weight=1),(Item=(Type=Bone,Name="pelvis"),Weight=1))
    Weight=1
   )
   > &RigUnit_ParentConstraint(
    AdvancedSettings=(InterpolationType=Average,RotationOrderForFilter=XZY)
    Child=(Type=Bone,Name="dyn_skirt_back_l_MSTR")
    Parents=((Item=(Type=Bone,Name="thigh_l"),Weight=1),(Item=(Type=Bone,Name="pelvis"),Weight=1),(Item=(Type=Bone,Name="thigh_r"),Weight=1))
    Weight=1
   )
   > &RigUnit_ParentConstraint(
    AdvancedSettings=(InterpolationType=Average,RotationOrderForFilter=XZY)
    Child=(Type=Bone,Name="dyn_skirt_back_r_MSTR")
    Parents=((Item=(Type=Bone,Name="thigh_r"),Weight=1),(Item=(Type=Bone,Name="pelvis"),Weight=1))
    Weight=1
   )
   > &RigUnit_ParentConstraint(
    AdvancedSettings=(InterpolationType=Average,RotationOrderForFilter=XZY)
    Child=(Type=Bone,Name="dyn_skirt_side_r_MSTR")
    Parents=((Item=(Type=Bone,Name="thigh_r"),Weight=1))
    Weight=1
   )
   > &RigUnit_ParentConstraint(
    AdvancedSettings=(InterpolationType=Average,RotationOrderForFilter=XZY)
    Child=(Type=Bone,Name="dyn_skirt_front_r_MSTR")
    Parents=((Item=(Type=Bone,Name="thigh_r"),Weight=1))
    Weight=1
   )
   n1 = RigVMRerouteNode(Value=())
   n2 = &RigVMFunction_ControlFlowBranch(ExecuteContext=@n1.Value, Condition=RigVMVariableNode() {variable=Emotes}.Value)
   n3 = &RigVMFunction_ControlFlowBranch(ExecuteContext=@n2.False, Condition=RigVMVariableNode() {variable=ON_ground}.Value)
   n4 = &RigUnit_SphericalPoseReader(
    ActiveRegionScaleFactors=(PositiveWidth=1,NegativeWidth=1,PositiveHeight=1,NegativeHeight=1)
    ActiveRegionSize=0.11
    Debug=(bDrawDebug=True,bDraw2D=False,bDrawLocalAxes=False,DebugScale=25,DebugSegments=20,DebugThickness=0.25)
    DriverAxis=(X=-1,Y=0,Z=0)
    DriverItem=(Type=Bone,Name="calf_l")
    FalloffRegionScaleFactors=(PositiveWidth=1,NegativeWidth=1,PositiveHeight=1,NegativeHeight=1)
    FalloffSize=0.22
    FlipHeightScaling=False
    FlipWidthScaling=False
    OptionalParentItem=(Type=Bone,Name="None")
    RotationOffset=(X=-90,Y=0,Z=-15)
    ExecutePin=@n3.True
   )
   n5 = &RigUnit_GetRelativeTransformForItem(Child=(Type=Bone,Name="pelvis"), Parent=(Type=Bone,Name="root"), bChildInitial=False, bParentInitial=False)
   n6 = &RigVMFunction_KalmanFloat(
    BufferSize=6
    Value=&RigVMFunction_MathDoubleMul(
     B=&RigVMFunction_MathDoubleRemap(SourceMaximum=85, SourceMinimum=80, TargetMaximum=1, TargetMinimum=0, Value=@n5.RelativeTransform.Translation.Z).Result
     A=@n4.OutputParam
    ).Result
   ) {template="KalmanFilter::Execute(in Value,in BufferSize,out Result)"}
   > &RigUnit_ModifyTransforms(
    ItemToModify=((Item=(Type=Bone,Name="dyn_skirt_back_l_MSTR"),Transform=(Rotation=(X=0,Y=-0.2588,Z=0,W=0.9659),Translation=(X=0,Y=0,Z=0),Scale3D=(X=1,Y=1,Z=1))))
    WeightMinimum=0
    Weight=@n6.Result
   )
   > &RigUnit_ModifyTransforms(
    ItemToModify=((Item=(Type=Bone,Name="dyn_skirt_back_mid_MSTR"),Transform=(Rotation=(X=0,Y=-0.2588,Z=0,W=0.9659),Translation=(X=0,Y=0,Z=0),Scale3D=(X=1,Y=1,Z=1))))
    WeightMinimum=0
    Weight=@n6.Result
   )
   > n7 = &RigUnit_SphericalPoseReader(
    ActiveRegionScaleFactors=(PositiveWidth=1,NegativeWidth=1,PositiveHeight=1,NegativeHeight=1)
    ActiveRegionSize=0.11
    Debug=(bDrawDebug=True,bDraw2D=False,bDrawLocalAxes=False,DebugScale=25,DebugSegments=20,DebugThickness=0.25)
    DriverAxis=(X=1,Y=0,Z=0)
    DriverItem=(Type=Bone,Name="calf_r")
    FalloffRegionScaleFactors=(PositiveWidth=1,NegativeWidth=1,PositiveHeight=1,NegativeHeight=1)
    FalloffSize=0.22
    FlipHeightScaling=False
    FlipWidthScaling=False
    OptionalParentItem=(Type=Bone,Name="None")
    RotationOffset=(X=90,Y=0,Z=-15)
   )
   n8 = &RigUnit_GetRelativeTransformForItem(Child=(Type=Bone,Name="pelvis"), Parent=(Type=Bone,Name="root"), bChildInitial=False, bParentInitial=False)
   n9 = &RigVMFunction_KalmanFloat(
    BufferSize=6
    Value=&RigVMFunction_MathDoubleMul(
     B=&RigVMFunction_MathDoubleRemap(SourceMaximum=85, SourceMinimum=80, TargetMaximum=1, TargetMinimum=0, Value=@n8.RelativeTransform.Translation.Z).Result
     A=@n7.OutputParam
    ).Result
   ) {template="KalmanFilter::Execute(in Value,in BufferSize,out Result)"}
   > &RigUnit_ModifyTransforms(
    ItemToModify=((Item=(Type=Bone,Name="dyn_skirt_back_r_MSTR"),Transform=(Rotation=(X=0,Y=-0.2588,Z=0,W=0.9659),Translation=(X=0,Y=0,Z=0),Scale3D=(X=1,Y=1,Z=1))))
    WeightMinimum=0
    Weight=@n9.Result
   )
   > &RigUnit_ModifyTransforms(
    ItemToModify=((Item=(Type=Bone,Name="dyn_skirt_back_mid_MSTR"),Transform=(Rotation=(X=0,Y=-0.2588,Z=0,W=0.9659),Translation=(X=0,Y=0,Z=0),Scale3D=(X=1,Y=1,Z=1))))
    WeightMinimum=0
    Weight=@n9.Result
   )
   &RigUnit_ModifyTransforms(
    ItemToModify=((Item=(Type=Bone,Name="dyn_skirt_front_mid_MSTR"),Transform=(Rotation=(X=0,Y=0,Z=0,W=1),Translation=(X=0,Y=0,Z=1),Scale3D=(X=1,Y=1,Z=1))))
    Weight=1
    WeightMinimum=0
    ExecutePin=@n2.True
   )
   > &RigUnit_ModifyTransforms(
    ItemToModify=((Item=(Type=Bone,Name="dyn_skirt_front_l_MSTR"),Transform=(Rotation=(X=0,Y=0,Z=0,W=1),Translation=(X=0,Y=0,Z=1),Scale3D=(X=1,Y=1,Z=1))))
    Weight=1
    WeightMinimum=0
   )
   > &RigUnit_ModifyTransforms(
    ItemToModify=((Item=(Type=Bone,Name="dyn_skirt_side_l_MSTR"),Transform=(Rotation=(X=0,Y=0,Z=0,W=1),Translation=(X=0,Y=0,Z=1),Scale3D=(X=1,Y=1,Z=1))))
    Weight=1
    WeightMinimum=0
   )
   > &RigUnit_ModifyTransforms(
    ItemToModify=((Item=(Type=Bone,Name="dyn_skirt_back_l_MSTR"),Transform=(Rotation=(X=0,Y=0,Z=0,W=1),Translation=(X=0,Y=0,Z=1),Scale3D=(X=1,Y=1,Z=1))))
    Weight=1
    WeightMinimum=0
   )
   > &RigUnit_ModifyTransforms(
    ItemToModify=((Item=(Type=Bone,Name="dyn_skirt_back_mid_MSTR"),Transform=(Rotation=(X=0,Y=0,Z=0,W=1),Translation=(X=0,Y=0,Z=1),Scale3D=(X=1,Y=1,Z=1))))
    Weight=1
    WeightMinimum=0
   )
   > &RigUnit_ModifyTransforms(
    ItemToModify=((Item=(Type=Bone,Name="dyn_skirt_back_r_MSTR"),Transform=(Rotation=(X=0,Y=0,Z=0,W=1),Translation=(X=0,Y=0,Z=1),Scale3D=(X=1,Y=1,Z=1))))
    Weight=1
    WeightMinimum=0
   )
   > &RigUnit_ModifyTransforms(
    ItemToModify=((Item=(Type=Bone,Name="dyn_skirt_side_r_MSTR"),Transform=(Rotation=(X=0,Y=0,Z=0,W=1),Translation=(X=0,Y=0,Z=1),Scale3D=(X=1,Y=1,Z=1))))
    Weight=1
    WeightMinimum=0
   )
   > &RigUnit_ModifyTransforms(
    ItemToModify=((Item=(Type=Bone,Name="dyn_skirt_front_r_MSTR"),Transform=(Rotation=(X=0,Y=0,Z=0,W=1),Translation=(X=0,Y=0,Z=1),Scale3D=(X=1,Y=1,Z=1))))
    Weight=1
    WeightMinimum=0
   )
   > n10 = &RigUnit_SphericalPoseReader(
    ActiveRegionScaleFactors=(PositiveWidth=1,NegativeWidth=1,PositiveHeight=1,NegativeHeight=1)
    ActiveRegionSize=0.13
    Debug=(bDrawDebug=true,bDraw2D=False,bDrawLocalAxes=False,DebugScale=25,DebugSegments=20,DebugThickness=0.25)
    DriverAxis=(X=1,Y=0,Z=0)
    DriverItem=(Type=Bone,Name="thigh_r")
    FalloffRegionScaleFactors=(PositiveWidth=1,NegativeWidth=1,PositiveHeight=1,NegativeHeight=1)
    FalloffSize=0.26
    FlipHeightScaling=False
    FlipWidthScaling=False
    OptionalParentItem=(Type=Bone,Name="None")
    RotationOffset=(X=-90,Y=0,Z=0)
   )
   > &RigUnit_ModifyTransforms(
    ItemToModify=((Item=(Type=Bone,Name="dyn_skirt_front_r_MSTR"),Transform=(Rotation=(X=0,Y=-0.0872,Z=0,W=0.9962),Translation=(X=0,Y=0,Z=2),Scale3D=(X=1,Y=1,Z=1))))
    WeightMinimum=0
    Weight=@n10.OutputParam
   )
   > &RigUnit_ModifyTransforms(
    ItemToModify=((Item=(Type=Bone,Name="dyn_skirt_front_mid_MSTR"),Transform=(Rotation=(X=0,Y=-0.0872,Z=0,W=0.9962),Translation=(X=0,Y=0,Z=2),Scale3D=(X=1,Y=1,Z=1))))
    WeightMinimum=0
    Weight=@n10.OutputParam
   )
   > n11 = &RigUnit_SphericalPoseReader(
    ActiveRegionScaleFactors=(PositiveWidth=1,NegativeWidth=1,PositiveHeight=1,NegativeHeight=1)
    ActiveRegionSize=0.13
    Debug=(bDrawDebug=true,bDraw2D=False,bDrawLocalAxes=False,DebugScale=25,DebugSegments=20,DebugThickness=0.25)
    DriverAxis=(X=-1,Y=0,Z=0)
    DriverItem=(Type=Bone,Name="thigh_l")
    FalloffRegionScaleFactors=(PositiveWidth=1,NegativeWidth=1,PositiveHeight=1,NegativeHeight=1)
    FalloffSize=0.26
    FlipHeightScaling=False
    FlipWidthScaling=False
    OptionalParentItem=(Type=Bone,Name="None")
    RotationOffset=(X=90,Y=0,Z=0)
   )
   > &RigUnit_ModifyTransforms(
    ItemToModify=((Item=(Type=Bone,Name="dyn_skirt_front_l_MSTR"),Transform=(Rotation=(X=0,Y=-0.0872,Z=0,W=0.9962),Translation=(X=0,Y=0,Z=2),Scale3D=(X=1,Y=1,Z=1))))
    WeightMinimum=0
    Weight=@n11.OutputParam
   )
   > &RigUnit_ModifyTransforms(
    ItemToModify=((Item=(Type=Bone,Name="dyn_skirt_front_mid_MSTR"),Transform=(Rotation=(X=0,Y=-0.0872,Z=0,W=0.9962),Translation=(X=0,Y=0,Z=2),Scale3D=(X=1,Y=1,Z=1))))
    WeightMinimum=0
    Weight=@n11.OutputParam
   )
   RigUnit_BeginExecution = &RigUnit_BeginExecution()
   RigVMFunction_Sequence = RigVMAggregateNode(ExecuteContext=@RigUnit_BeginExecution.ExecutePin)
   FN_DeformRig_FNC = RigVMFunctionReferenceNode(
    DeformRig_ Curve="(deformRig_l_upperarm_up_alpha_72_#1=False,deformRig_r_upperarm_up_alpha_73_#2=False,deformRig_l_upperarm_fr_alpha_74_#3=False,deformRig_r_upperarm_fr_alpha_75_CB51B8 ...(truncated)"
    DeformRig_Bool="(upperarm_1_#4=True,elbow_3_#5=True,wrist_7_#6=True,groin_8_#7=True,glute_17_#8=True,knee_9_40C408324CA261 ...(truncated)"
    DeformRig_Debug=(upperarm_debug_18_#4=False,wrist_debug_20_#6=False,lat_debug_26_#9=False,groinGlute_debug_27_#7=False)
    DeformRig_Value="(upperarm_up_18_#10=(Rotation=(X=0,Y=0,Z=0,W=1),Translation=(X=0,Y=0,Z=8),Scale3D=(X=0.7,Y=1,Z=1)),upperarm_fwd_fr_19_#11=(Rotation=(X=0,Y=0,Z=0,W=1),Translation=(X=0,Y=10,Z=0),Scale3D=(X=0.5,Y=1,Z= ...(truncated)"
    ExecuteContext=@RigVMFunction_Sequence.A
   )
   n12 = &RigVMFunction_AlphaInterp(InterpSpeedDecreasing=20, InterpSpeedIncreasing=20, Scale=1, bClampResult=False, Value=RigVMVariableNode() {variable=leg_adjust_alpha}.Value)
   > &RigUnit_ModifyTransforms(
    ItemToModify="((Item=(Type=Bone,Name=\"deform_thigh_front_l\"),Transform=(Rotation=(X=0,Y=0,Z=0,W=-1),Translation=(X=-5,Y=0,Z=0),Scale3D=(X=1,Y=1,Z=1))),(Item=(Type=Bone,Name=\"deform_thigh_front_r\"),Transform=(Rotation=(X=0,Y=0,Z=0,W=-1),Translation=(X=-5,Y=0,Z=0),Scale3D ...(truncated)"
    WeightMinimum=0
    Weight=@n12.Result
   )
   &RigUnit_ModifyTransforms(
    ItemToModify="((Item=(Type=Bone,Name=\"dyn_lace_l_end_l_MSTR\"),Transform=(Rotation=(X=-0.0042,Y=0.0696,Z=0.0602,W=-0.9957),Translation=(X=0,Y=0,Z=0),Scale3D=(X=1,Y=1,Z=1))),(Item=(Type=Bone,Name=\"dyn_lace_l_loop_l_MSTR\"),Transform=(Rotation=(X=0,Y=-0.0728,Z=0,W=0.9973),T ...(truncated)"
    Weight=1
    WeightMinimum=0
    ExecutePin=@RigVMFunction_Sequence.B
   )
   > &RigUnit_ModifyTransforms(
    ItemToModify="((Item=(Type=Bone,Name=\"dyn_lace_r_end_l_MSTR\"),Transform=(Rotation=(X=0.0288,Y=-0.0635,Z=-0.0676,W=0.9953),Translation=(X=0,Y=0,Z=0),Scale3D=(X=1,Y=1,Z=1))),(Item=(Type=Bone,Name=\"dyn_lace_r_loop_l_MSTR\"),Transform=(Rotation=(X=0,Y=-0.0728,Z=0,W=0.9973),T ...(truncated)"
    Weight=1
    WeightMinimum=0
   )
   Multiply = &RigVMFunction_MathTransformMul(
    B=/Script/RigVM.RigVMFunction_MathTransformInverse(Value=&RigUnit_GetTransform(Item=(Type=Bone,Name="thigh_l"), Space=LocalSpace, bInitial=true).Transform) {is=struct RigVMUnitNode, template="Inverse::Execute(in Value,out Result)"}.Result
    A=&RigUnit_GetTransform(Item=(Type=Bone,Name="thigh_l"), Space=LocalSpace, bInitial=False).Transform
   ) {template="Multiply::Execute(in A,in B,out Result)"}
   RigVMFunction_MathQuaternionSwingTwist = &RigVMFunction_MathQuaternionSwingTwist(TwistAxis=(X=1,Y=0,Z=0), Input=@Multiply.Result.Rotation)
   Inverse = /Script/RigVM.RigVMFunction_MathQuaternionInverse(Value=@RigVMFunction_MathQuaternionSwingTwist.Twist) {is=struct RigVMUnitNode, template="Inverse::Execute(in Value,out Result)"}
   RigUnit_SetRotation = &RigUnit_SetRotation(Item=(Type=Bone,Name="CR_thigh_l_no_twist"), Weight=1, bInitial=False, Value=@Inverse.Result, ExecutePin=@RigVMFunction_Sequence.C)
   > &RigUnit_ParentConstraint(
    AdvancedSettings=(InterpolationType=Average,RotationOrderForFilter=XZY)
    Child=(Type=Bone,Name="CR_thigh_r")
    Parents=((Item=(Type=Bone,Name="thigh_r"),Weight=1))
    Weight=1
    bMaintainOffset=false
   )
   n13 = &RigVMFunction_MathTransformMul(
    B=&RigVMFunction_MathTransformInverse(Value=&RigUnit_GetTransform(Item=(Type=Bone,Name="thigh_r"), Space=LocalSpace, bInitial=true).Transform) {template="Inverse::Execute(in Value,out Result)"}.Result
    A=&RigUnit_GetTransform(Item=(Type=Bone,Name="thigh_r"), Space=LocalSpace, bInitial=False).Transform
   ) {template="Multiply::Execute(in A,in B,out Result)"}
   > &RigUnit_SetRotation(
    Item=(Type=Bone,Name="CR_thigh_r_no_twist")
    Weight=1
    bInitial=False
    Value=&RigVMFunction_MathQuaternionInverse(Value=&RigVMFunction_MathQuaternionSwingTwist(TwistAxis=(X=-1,Y=0,Z=0), Input=@n13.Result.Rotation).Twist) {template="Inverse::Execute(in Value,out Result)"}.Result
   )
   n14 = &RigVMFunction_ControlFlowBranch(Condition=RigVMVariableNode() {variable=TombstoneEmoteActive}.Value, ExecuteContext=@RigVMFunction_Sequence.E)
   RigVMFunction_Sequence_1 = RigVMAggregateNode(ExecuteContext=@n14.Completed)
   n15 = &RigUnit_SphericalPoseReader(
    ActiveRegionScaleFactors=(PositiveWidth=3,NegativeWidth=1,PositiveHeight=1.2,NegativeHeight=1)
    ActiveRegionSize=0.1
    Debug=(bDrawDebug=false,bDraw2D=False,bDrawLocalAxes=False,DebugScale=25,DebugSegments=20,DebugThickness=0.25)
    DriverAxis=(X=-1,Y=0,Z=0)
    DriverItem=(Type=Bone,Name="thigh_l")
    FalloffRegionScaleFactors=(PositiveWidth=1,NegativeWidth=1,PositiveHeight=0.75,NegativeHeight=0.75)
    FalloffSize=0.58
    FlipHeightScaling=False
    FlipWidthScaling=False
    OptionalParentItem=(Type=Bone,Name="pelvis")
    RotationOffset=(X=90,Y=0,Z=-15)
    ExecutePin=@RigVMFunction_Sequence_1.A
   )
   > RigVMVariableNode(Value=@n15.OutputParam) {variable=thigh_l_up_out}
   > &RigUnit_ModifyTransforms(
    ItemToModify="((Item=(Type=Bone,Name=\"dyn_skirt_front_mid_MSTR\"),Transform=(Rotation=(X=-0.0868,Y=0.0076,Z=-0.0868,W=0.9924),Translation=(X=0,Y=0,Z=0),Scale3D=(X=1,Y=1,Z=1))),(Item=(Type=Bone,Name=\"dyn_skirt_side_l_MSTR\"),Transform=(Rotation=(X=0,Y=0,Z=-0.1736,W=0.9848) ...(truncated)"
    WeightMinimum=0
    Weight=RigVMVariableNode() {variable=thigh_l_up_out}.Value
   )
   > n16 = &RigUnit_SphericalPoseReader(
    ActiveRegionScaleFactors=(PositiveWidth=3,NegativeWidth=1,PositiveHeight=1.3,NegativeHeight=1.2)
    ActiveRegionSize=0.1
    Debug=(bDrawDebug=false,bDraw2D=False,bDrawLocalAxes=False,DebugScale=25,DebugSegments=20,DebugThickness=0.25)
    DriverAxis=(X=-1,Y=0,Z=0)
    DriverItem=(Type=Bone,Name="thigh_l")
    FalloffRegionScaleFactors=(PositiveWidth=1,NegativeWidth=1,PositiveHeight=0.5,NegativeHeight=0.5)
    FalloffSize=0.57
    FlipHeightScaling=False
    FlipWidthScaling=False
    OptionalParentItem=(Type=Bone,Name="pelvis")
    RotationOffset=(X=130,Y=5,Z=-12)
   )
   > RigVMVariableNode(Value=@n16.OutputParam) {variable=thigh_l_up_side_inb_out}
   > &RigUnit_ModifyTransforms(
    ItemToModify="((Item=(Type=Bone,Name=\"dyn_skirt_front_mid_MSTR\"),Transform=(Rotation=(X=0.0006,Y=-0.0174,Z=0.0349,W=0.9992),Translation=(X=0,Y=0,Z=2),Scale3D=(X=1,Y=1,Z=1))),(Item=(Type=Bone,Name=\"dyn_skirt_side_l_MSTR\"),Transform=(Rotation=(X=0,Y=0,Z=0.1736,W=0.9848),T ...(truncated)"
    WeightMinimum=0
    Weight=RigVMVariableNode() {variable=thigh_l_up_side_inb_out}.Value
   )
   > n17 = &RigUnit_SphericalPoseReader(
    ActiveRegionScaleFactors=(PositiveWidth=3,NegativeWidth=1,PositiveHeight=1,NegativeHeight=1)
    ActiveRegionSize=0.15
    Debug=(bDrawDebug=false,bDraw2D=False,bDrawLocalAxes=False,DebugScale=25,DebugSegments=20,DebugThickness=0.25)
    DriverAxis=(X=-1,Y=0,Z=0)
    DriverItem=(Type=Bone,Name="thigh_l")
    FalloffRegionScaleFactors=(PositiveWidth=1,NegativeWidth=1,PositiveHeight=1,NegativeHeight=1)
    FalloffSize=0.5
    FlipHeightScaling=False
    FlipWidthScaling=False
    OptionalParentItem=(Type=Bone,Name="pelvis")
    RotationOffset=(X=180,Y=0,Z=0)
   )
   > RigVMVariableNode(Value=@n17.OutputParam) {variable=thigh_l_side_out}
   > &RigUnit_ModifyTransforms(
    ItemToModify="((Item=(Type=Bone,Name=\"dyn_skirt_front_mid_MSTR\"),Transform=(Rotation=(X=0.1137,Y=0.058,Z=0.0013,W=-0.9918),Translation=(X=0,Y=0,Z=0),Scale3D=(X=1,Y=1,Z=1))),(Item=(Type=Bone,Name=\"dyn_skirt_side_l_MSTR\"),Transform=(Rotation=(X=0.2054,Y=-0.3195,Z=-0.1153, ...(truncated)"
    WeightMinimum=0
    Weight=RigVMVariableNode() {variable=thigh_l_side_out}.Value
   )
   > n18 = &RigUnit_SphericalPoseReader(
    ActiveRegionScaleFactors=(PositiveWidth=3,NegativeWidth=1,PositiveHeight=1.5,NegativeHeight=1.5)
    ActiveRegionSize=0.1
    Debug=(bDrawDebug=false,bDraw2D=False,bDrawLocalAxes=False,DebugScale=25,DebugSegments=20,DebugThickness=0.25)
    DriverAxis=(X=-1,Y=0,Z=0)
    DriverItem=(Type=Bone,Name="thigh_l")
    FalloffRegionScaleFactors=(PositiveWidth=1,NegativeWidth=1,PositiveHeight=1,NegativeHeight=1)
    FalloffSize=0.5
    FlipHeightScaling=False
    FlipWidthScaling=False
    OptionalParentItem=(Type=Bone,Name="pelvis")
    RotationOffset=(X=-125,Y=0,Z=0)
   )
   > RigVMVariableNode(Value=@n18.OutputParam) {variable=thigh_l_side_back_inb_out}
   > &RigUnit_ModifyTransforms(
    ItemToModify="((Item=(Type=Bone,Name=\"dyn_skirt_front_mid_MSTR\"),Transform=(Rotation=(X=0,Y=0,Z=0,W=1),Translation=(X=0,Y=0,Z=0),Scale3D=(X=1,Y=1,Z=1))),(Item=(Type=Bone,Name=\"dyn_skirt_side_l_MSTR\"),Transform=(Rotation=(X=0.0016,Y=-0.0659,Z=0.024,W=0.9975),Translation= ...(truncated)"
    WeightMinimum=0
    Weight=RigVMVariableNode() {variable=thigh_l_side_back_inb_out}.Value
   )
   > n19 = &RigUnit_SphericalPoseReader(
    ActiveRegionScaleFactors=(PositiveWidth=3,NegativeWidth=1,PositiveHeight=3,NegativeHeight=3)
    ActiveRegionSize=0.05
    Debug=(bDrawDebug=false,bDraw2D=False,bDrawLocalAxes=False,DebugScale=25,DebugSegments=20,DebugThickness=0.25)
    DriverAxis=(X=-1,Y=0,Z=0)
    DriverItem=(Type=Bone,Name="thigh_l")
    FalloffRegionScaleFactors=(PositiveWidth=1,NegativeWidth=0.95,PositiveHeight=1,NegativeHeight=1)
    FalloffSize=0.3
    FlipHeightScaling=False
    FlipWidthScaling=False
    OptionalParentItem=(Type=Bone,Name="pelvis")
    RotationOffset=(X=-90,Y=0,Z=-30)
   )
   RigVMFunction_AlphaInterp = &RigVMFunction_AlphaInterp(InterpSpeedDecreasing=10, InterpSpeedIncreasing=10, Scale=1, bClampResult=False, bInterpResult=True, Value=@n19.OutputParam)
   > RigVMVariableNode(Value=@RigVMFunction_AlphaInterp.Result) {variable=thigh_l_back_out}
   > &RigUnit_ModifyTransforms(
    ItemToModify="((Item=(Type=Bone,Name=\"dyn_skirt_front_mid_MSTR\"),Transform=(Rotation=(X=0.0405,Y=-0.2348,Z=-0.0382,W=0.9704),Translation=(X=0,Y=0,Z=0),Scale3D=(X=1,Y=1,Z=1))),(Item=(Type=Bone,Name=\"dyn_skirt_side_l_MSTR\"),Transform=(Rotation=(X=-0.0056,Y=-0.1736,Z=-0.03 ...(truncated)"
    WeightMinimum=0
    Weight=RigVMVariableNode() {variable=thigh_l_back_out}.Value
   )
   > n20 = &RigUnit_SphericalPoseReader(
    ActiveRegionScaleFactors=(PositiveWidth=3,NegativeWidth=1,PositiveHeight=1.2,NegativeHeight=1)
    ActiveRegionSize=0.1
    Debug=(bDrawDebug=false,bDraw2D=False,bDrawLocalAxes=False,DebugScale=25,DebugSegments=20,DebugThickness=0.25)
    DriverAxis=(X=-1,Y=0,Z=0)
    DriverItem=(Type=Bone,Name="thigh_l")
    FalloffRegionScaleFactors=(PositiveWidth=1,NegativeWidth=1,PositiveHeight=0.3,NegativeHeight=0.3)
    FalloffSize=0.58
    FlipHeightScaling=False
    FlipWidthScaling=False
    OptionalParentItem=(Type=Bone,Name="pelvis")
    RotationOffset=(X=45,Y=0,Z=-15)
   )
   > RigVMVariableNode(Value=@n20.OutputParam) {variable=thigh_l_up_across_out}
   > &RigUnit_ModifyTransforms(
    ItemToModify="((Item=(Type=Bone,Name=\"dyn_skirt_front_mid_MSTR\"),Transform=(Rotation=(X=0.4312,Y=0.1064,Z=-0.0267,W=-0.8955),Translation=(X=3.5813,Y=6.2581,Z=4.5909),Scale3D=(X=1,Y=1,Z=1))),(Item=(Type=Bone,Name=\"dyn_skirt_side_l_MSTR\"),Transform=(Rotation=(X=0.0373,Y=- ...(truncated)"
    WeightMinimum=0
    Weight=RigVMVariableNode() {variable=thigh_l_up_across_out}.Value
   )
   > n21 = &RigUnit_SphericalPoseReader(
    ActiveRegionScaleFactors=(PositiveWidth=1,NegativeWidth=1,PositiveHeight=1,NegativeHeight=1)
    ActiveRegionSize=0.2
    Debug=(bDrawDebug=false,bDraw2D=False,bDrawLocalAxes=False,DebugScale=25,DebugSegments=20,DebugThickness=0.25)
    DriverAxis=(X=-1,Y=0,Z=0)
    DriverItem=(Type=Bone,Name="thigh_l")
    FalloffRegionScaleFactors=(PositiveWidth=1,NegativeWidth=1,PositiveHeight=1,NegativeHeight=1)
    FalloffSize=0.3
    FlipHeightScaling=False
    FlipWidthScaling=False
    OptionalParentItem=(Type=Bone,Name="pelvis")
    RotationOffset=(X=90,Y=0,Z=10)
   )
   > /Script/ControlRig.RigUnit_SetCurveValue(Curve=thigh_l_up_SPR, Value=@n21.OutputParam) {is=struct RigVMUnitNode}
   > n22 = &RigUnit_SphericalPoseReader(
    ActiveRegionScaleFactors=(PositiveWidth=1,NegativeWidth=1,PositiveHeight=1.5,NegativeHeight=1.5)
    ActiveRegionSize=0.1
    Debug=(bDrawDebug=false,bDraw2D=False,bDrawLocalAxes=False,DebugScale=25,DebugSegments=20,DebugThickness=0.25)
    DriverAxis=(X=-1,Y=0,Z=0)
    DriverItem=(Type=Bone,Name="calf_l")
    FalloffRegionScaleFactors=(PositiveWidth=1,NegativeWidth=1,PositiveHeight=1,NegativeHeight=1)
    FalloffSize=0.35
    FlipHeightScaling=False
    FlipWidthScaling=False
    OptionalParentItem=(Type=Bone,Name="thigh_l")
    RotationOffset=(X=90,Y=0,Z=180)
   )
   n23 = &RigVMFunction_AlphaInterp(InterpSpeedDecreasing=10, InterpSpeedIncreasing=10, Scale=1, bClampResult=False, bInterpResult=True, Value=@n22.OutputParam)
   > RigVMVariableNode(Value=@n23.Result) {variable=calf_l_up_out}
   > &RigUnit_ModifyTransforms(
    ItemToModify="((Item=(Type=Bone,Name=\"dyn_skirt_back_l_MSTR\"),Transform=(Rotation=(X=0,Y=-0.1305,Z=0,W=0.9914),Translation=(X=0,Y=0,Z=0),Scale3D=(X=1,Y=1,Z=1))),(Item=(Type=Bone,Name=\"dyn_skirt_back_mid_MSTR\"),Transform=(Rotation=(X=0,Y=-0.1305,Z=0,W=0.9914),Translation ...(truncated)"
    WeightMinimum=0
    Weight=RigVMVariableNode() {variable=calf_l_up_out}.Value
   )
   n24 = &RigUnit_SphericalPoseReader(
    ActiveRegionScaleFactors=(PositiveWidth=3,NegativeWidth=1,PositiveHeight=1.2,NegativeHeight=1)
    ActiveRegionSize=0.1
    Debug=(bDrawDebug=false,bDraw2D=False,bDrawLocalAxes=False,DebugScale=25,DebugSegments=20,DebugThickness=0.25)
    DriverAxis=(X=1,Y=0,Z=0)
    DriverItem=(Type=Bone,Name="thigh_r")
    FalloffRegionScaleFactors=(PositiveWidth=1,NegativeWidth=1,PositiveHeight=0.75,NegativeHeight=0.75)
    FalloffSize=0.58
    FlipHeightScaling=true
    FlipWidthScaling=true
    OptionalParentItem=(Type=Bone,Name="None")
    RotationOffset=(X=-90,Y=0,Z=-15)
    ExecutePin=@RigVMFunction_Sequence_1.B
   )
   > RigVMVariableNode(Value=@n24.OutputParam) {variable=thigh_r_up_out}
   > &RigUnit_ModifyTransforms(
    ItemToModify="((Item=(Type=Bone,Name=\"dyn_skirt_front_mid_MSTR\"),Transform=(Rotation=(X=0.0871,Y=-0.0435,Z=0.0038,W=0.9952),Translation=(X=2.5393,Y=-0.8918,Z=1.1519),Scale3D=(X=1,Y=1,Z=1))),(Item=(Type=Bone,Name=\"dyn_skirt_side_r_MSTR\"),Transform=(Rotation=(X=0.0302,Y=- ...(truncated)"
    WeightMinimum=0
    Weight=RigVMVariableNode() {variable=thigh_r_up_out}.Value
   )
   > n25 = &RigUnit_SphericalPoseReader(
    ActiveRegionScaleFactors=(PositiveWidth=3,NegativeWidth=1,PositiveHeight=1.3,NegativeHeight=1.2)
    ActiveRegionSize=0.1
    Debug=(bDrawDebug=false,bDraw2D=False,bDrawLocalAxes=False,DebugScale=25,DebugSegments=20,DebugThickness=0.25)
    DriverAxis=(X=1,Y=0,Z=0)
    DriverItem=(Type=Bone,Name="thigh_r")
    FalloffRegionScaleFactors=(PositiveWidth=1,NegativeWidth=1,PositiveHeight=0.5,NegativeHeight=0.5)
    FalloffSize=0.57
    FlipHeightScaling=true
    FlipWidthScaling=true
    OptionalParentItem=(Type=Bone,Name="None")
    RotationOffset=(X=-50,Y=5,Z=-12)
   )
   > RigVMVariableNode(Value=@n25.OutputParam) {variable=thigh_r_up_side_inb_out}
   > &RigUnit_ModifyTransforms(
    ItemToModify="((Item=(Type=Bone,Name=\"dyn_skirt_front_mid_MSTR\"),Transform=(Rotation=(X=0,Y=0,Z=-0.0349,W=0.9994),Translation=(X=0,Y=0,Z=0),Scale3D=(X=1,Y=1,Z=1))),(Item=(Type=Bone,Name=\"dyn_skirt_side_r_MSTR\"),Transform=(Rotation=(X=0.0645,Y=-0.0102,Z=0.0855,W=-0.9942) ...(truncated)"
    WeightMinimum=0
    Weight=RigVMVariableNode() {variable=thigh_r_up_side_inb_out}.Value
   )
   > n26 = &RigUnit_SphericalPoseReader(
    ActiveRegionScaleFactors=(PositiveWidth=3,NegativeWidth=1,PositiveHeight=1,NegativeHeight=1)
    ActiveRegionSize=0.1
    Debug=(bDrawDebug=false,bDraw2D=False,bDrawLocalAxes=False,DebugScale=25,DebugSegments=20,DebugThickness=0.25)
    DriverAxis=(X=1,Y=0,Z=0)
    DriverItem=(Type=Bone,Name="thigh_r")
    FalloffRegionScaleFactors=(PositiveWidth=1,NegativeWidth=1,PositiveHeight=1,NegativeHeight=1)
    FalloffSize=0.5
    FlipHeightScaling=true
    FlipWidthScaling=true
    OptionalParentItem=(Type=Bone,Name="None")
    RotationOffset=(X=0,Y=0,Z=0)
   )
   > RigVMVariableNode(Value=@n26.OutputParam) {variable=thigh_r_side_out}
   > &RigUnit_ModifyTransforms(
    ItemToModify="((Item=(Type=Bone,Name=\"dyn_skirt_front_mid_MSTR\"),Transform=(Rotation=(X=0,Y=0,Z=-0.0195,W=-0.9998),Translation=(X=0,Y=0,Z=0),Scale3D=(X=1,Y=1,Z=1))),(Item=(Type=Bone,Name=\"dyn_skirt_side_r_MSTR\"),Transform=(Rotation=(X=0,Y=0,Z=0.1463,W=0.9892),Translatio ...(truncated)"
    WeightMinimum=0
    Weight=RigVMVariableNode() {variable=thigh_r_side_out}.Value
   )
   > n27 = &RigUnit_SphericalPoseReader(
    ActiveRegionScaleFactors=(PositiveWidth=3,NegativeWidth=1,PositiveHeight=1.5,NegativeHeight=1.5)
    ActiveRegionSize=0.1
    Debug=(bDrawDebug=false,bDraw2D=False,bDrawLocalAxes=False,DebugScale=25,DebugSegments=20,DebugThickness=0.25)
    DriverAxis=(X=1,Y=0,Z=0)
    DriverItem=(Type=Bone,Name="thigh_r")
    FalloffRegionScaleFactors=(PositiveWidth=1,NegativeWidth=1,PositiveHeight=1,NegativeHeight=1)
    FalloffSize=0.5
    FlipHeightScaling=true
    FlipWidthScaling=true
    OptionalParentItem=(Type=Bone,Name="None")
    RotationOffset=(X=55,Y=0,Z=0)
   )
   > RigVMVariableNode(Value=@n27.OutputParam) {variable=thigh_r_side_back_inb_out}
   > &RigUnit_ModifyTransforms(
    ItemToModify="((Item=(Type=Bone,Name=\"dyn_skirt_front_mid_MSTR\"),Transform=(Rotation=(X=-0.0232,Y=-0.1985,Z=0.0537,W=0.9783),Translation=(X=0,Y=0,Z=0),Scale3D=(X=1,Y=1,Z=1))),(Item=(Type=Bone,Name=\"dyn_skirt_side_r_MSTR\"),Transform=(Rotation=(X=0,Y=0,Z=0.0176,W=0.9998), ...(truncated)"
    WeightMinimum=0
    Weight=RigVMVariableNode() {variable=thigh_r_side_back_inb_out}.Value
   )
   > n28 = &RigUnit_SphericalPoseReader(
    ActiveRegionScaleFactors=(PositiveWidth=3,NegativeWidth=1,PositiveHeight=3,NegativeHeight=3)
    ActiveRegionSize=0.05
    Debug=(bDrawDebug=false,bDraw2D=False,bDrawLocalAxes=False,DebugScale=25,DebugSegments=20,DebugThickness=0.25)
    DriverAxis=(X=1,Y=0,Z=0)
    DriverItem=(Type=Bone,Name="thigh_r")
    FalloffRegionScaleFactors=(PositiveWidth=1,NegativeWidth=0.95,PositiveHeight=1,NegativeHeight=1)
    FalloffSize=0.3
    FlipHeightScaling=true
    FlipWidthScaling=true
    OptionalParentItem=(Type=Bone,Name="None")
    RotationOffset=(X=90,Y=0,Z=-30)
   )
   > RigVMVariableNode() {variable=thigh_r_back_out}
   > &RigUnit_ModifyTransforms(
    ItemToModify="((Item=(Type=Bone,Name=\"dyn_skirt_front_mid_MSTR\"),Transform=(Rotation=(X=0,Y=-0.09,Z=0,W=0.9959),Translation=(X=-0.8698,Y=0,Z=4.7742),Scale3D=(X=1,Y=1,Z=1))),(Item=(Type=Bone,Name=\"dyn_skirt_side_r_MSTR\"),Transform=(Rotation=(X=-0.0028,Y=0.0699,Z=0.0404,W ...(truncated)"
    WeightMinimum=0
    Weight=RigVMVariableNode() {variable=thigh_r_back_out}.Value
   )
   > n29 = &RigUnit_SphericalPoseReader(
    ActiveRegionScaleFactors=(PositiveWidth=3,NegativeWidth=1,PositiveHeight=1.2,NegativeHeight=1)
    ActiveRegionSize=0.1
    Debug=(bDrawDebug=false,bDraw2D=False,bDrawLocalAxes=False,DebugScale=25,DebugSegments=20,DebugThickness=0.25)
    DriverAxis=(X=1,Y=0,Z=0)
    DriverItem=(Type=Bone,Name="thigh_r")
    FalloffRegionScaleFactors=(PositiveWidth=1,NegativeWidth=1,PositiveHeight=0.3,NegativeHeight=0.3)
    FalloffSize=0.58
    FlipHeightScaling=true
    FlipWidthScaling=true
    OptionalParentItem=(Type=Bone,Name="None")
    RotationOffset=(X=-135,Y=0,Z=-15)
   )
   > RigVMVariableNode(Value=@n29.OutputParam) {variable=thigh_r_up_across_out}
   > &RigUnit_ModifyTransforms(
    ItemToModify="((Item=(Type=Bone,Name=\"dyn_skirt_front_mid_MSTR\"),Transform=(Rotation=(X=0.5421,Y=-0.0075,Z=-0.0116,W=0.8402),Translation=(X=-1.3673,Y=-8.341,Z=-0.582),Scale3D=(X=1,Y=1,Z=1))),(Item=(Type=Bone,Name=\"dyn_skirt_side_r_MSTR\"),Transform=(Rotation=(X=-0.0088,Y ...(truncated)"
    WeightMinimum=0
    Weight=RigVMVariableNode() {variable=thigh_r_up_across_out}.Value
   )
   > n30 = &RigUnit_SphericalPoseReader(
    ActiveRegionScaleFactors=(PositiveWidth=1,NegativeWidth=1,PositiveHeight=1,NegativeHeight=1)
    ActiveRegionSize=0.2
    Debug=(bDrawDebug=false,bDraw2D=False,bDrawLocalAxes=False,DebugScale=25,DebugSegments=20,DebugThickness=0.25)
    DriverAxis=(X=1,Y=0,Z=0)
    DriverItem=(Type=Bone,Name="thigh_r")
    FalloffRegionScaleFactors=(PositiveWidth=1,NegativeWidth=1,PositiveHeight=1,NegativeHeight=1)
    FalloffSize=0.4
    FlipHeightScaling=False
    FlipWidthScaling=False
    OptionalParentItem=(Type=Bone,Name="pelvis")
    RotationOffset=(X=90,Y=0,Z=190)
   )
   > &RigUnit_SetCurveValue(Curve=thigh_r_up_SPR, Value=@n30.OutputParam)
   > n31 = &RigUnit_SphericalPoseReader(
    ActiveRegionScaleFactors=(PositiveWidth=1,NegativeWidth=1,PositiveHeight=1.5,NegativeHeight=1.5)
    ActiveRegionSize=0.1
    Debug=(bDrawDebug=false,bDraw2D=False,bDrawLocalAxes=False,DebugScale=25,DebugSegments=20,DebugThickness=0.25)
    DriverAxis=(X=1,Y=0,Z=0)
    DriverItem=(Type=Bone,Name="calf_r")
    FalloffRegionScaleFactors=(PositiveWidth=1,NegativeWidth=1,PositiveHeight=1,NegativeHeight=1)
    FalloffSize=0.35
    FlipHeightScaling=False
    FlipWidthScaling=False
    OptionalParentItem=(Type=Bone,Name="thigh_l")
    RotationOffset=(X=90,Y=0,Z=0)
   )
   n32 = &RigVMFunction_AlphaInterp(InterpSpeedDecreasing=10, InterpSpeedIncreasing=10, Scale=1, bClampResult=False, bInterpResult=True, Value=@n31.OutputParam)
   > RigVMVariableNode(Value=@n32.Result) {variable=calf_r_up_out}
   > &RigUnit_ModifyTransforms(
    ItemToModify="((Item=(Type=Bone,Name=\"dyn_skirt_back_r_MSTR\"),Transform=(Rotation=(X=0,Y=-0.1305,Z=0,W=0.9914),Translation=(X=0,Y=0,Z=0),Scale3D=(X=1,Y=1,Z=1))),(Item=(Type=Bone,Name=\"dyn_skirt_back_mid_MSTR\"),Transform=(Rotation=(X=0,Y=-0.1305,Z=0,W=0.9914),Translation ...(truncated)"
    WeightMinimum=0
    Weight=RigVMVariableNode() {variable=calf_r_up_out}.Value
   )
   &RigUnit_ModifyTransforms(
    ItemToModify="((Item=(Type=Bone,Name=\"dyn_skirt_back_l_MSTR\"),Transform=(Rotation=(X=0,Y=0.0402,Z=0,W=0.9992),Translation=(X=0,Y=0,Z=0),Scale3D=(X=1,Y=1,Z=1))),(Item=(Type=Bone,Name=\"dyn_skirt_back_mid_MSTR\"),Transform=(Rotation=(X=-0.0006,Y=-0.4181,Z=-0.0003,W=-0.9084) ...(truncated)"
    WeightMinimum=0
    Weight=RigVMVariableNode() {variable=TombstoneAlpha}.Value
    ExecutePin=@RigVMFunction_Sequence_1.C
   )
   RigVMVariableNode(ExecuteContext=@n14.False) {variable=TombstoneAlpha}
   RigVMVariableNode(ExecuteContext=@n14.True) {variable=thigh_l_up_out}
   > RigVMVariableNode() {variable=thigh_l_up_side_inb_out}
   > RigVMVariableNode() {variable=thigh_l_side_out}
   > RigVMVariableNode() {variable=thigh_l_side_back_inb_out}
   > RigVMVariableNode() {variable=thigh_l_back_out}
   > RigVMVariableNode() {variable=thigh_l_up_across_out}
   > RigVMVariableNode() {variable=thigh_r_up_out}
   > RigVMVariableNode() {variable=thigh_r_up_side_inb_out}
   > RigVMVariableNode() {variable=thigh_r_side_out}
   > RigVMVariableNode() {variable=thigh_r_side_back_inb_out}
   > RigVMVariableNode() {variable=thigh_r_back_out}
   > RigVMVariableNode() {variable=calf_l_up_out}
   > RigVMVariableNode() {variable=calf_r_up_out}
   > RigVMVariableNode() {variable=thigh_r_up_across_out}
   &RigUnit_ParentConstraint(
    AdvancedSettings=(InterpolationType=Average,RotationOrderForFilter=XZY)
    Child=(Type=Bone,Name="dyn_skirt_front_l_MSTR")
    Parents=((Item=(Type=Bone,Name="CR_thigh_l_no_twist"),Weight=1))
    Weight=1
    ExecutePin=@RigVMFunction_Sequence.D
   )
   > &RigUnit_ParentConstraint(
    AdvancedSettings=(InterpolationType=Average,RotationOrderForFilter=XZY)
    Child=(Type=Bone,Name="dyn_skirt_side_l_MSTR")
    Parents=((Item=(Type=Bone,Name="CR_thigh_l_no_twist"),Weight=0.33),(Item=(Type=Bone,Name="dyn_skirt_front_l_MSTR"),Weight=0.33),(Item=(Type=Bone,Name="dyn_skirt_front_mid_MSTR"),Weight=0.33))
    Weight=1
   )
   > &RigUnit_ParentConstraint(
    AdvancedSettings=(InterpolationType=Average,RotationOrderForFilter=XZY)
    Child=(Type=Bone,Name="dyn_skirt_front_r_MSTR")
    Parents=((Item=(Type=Bone,Name="CR_thigh_r_no_twist"),Weight=1))
    Weight=1
   )
   > &RigUnit_ParentConstraint(
    AdvancedSettings=(InterpolationType=Average,RotationOrderForFilter=XZY)
    Child=(Type=Bone,Name="dyn_skirt_side_r_MSTR")
    Parents=((Item=(Type=Bone,Name="CR_thigh_r_no_twist"),Weight=0.33),(Item=(Type=Bone,Name="dyn_skirt_front_r_MSTR"),Weight=0.33),(Item=(Type=Bone,Name="dyn_skirt_front_mid_MSTR"),Weight=0.33))
    Weight=1
   )
   > &RigUnit_ParentConstraint(
    AdvancedSettings=(InterpolationType=Average,RotationOrderForFilter=XZY)
    Child=(Type=Bone,Name="dyn_skirt_front_mid_MSTR")
    Parents=((Item=(Type=Bone,Name="dyn_skirt_front_l_MSTR"),Weight=0.5),(Item=(Type=Bone,Name="dyn_skirt_front_r_MSTR"),Weight=0.5))
    Weight=1
   )
   > &RigUnit_ParentConstraint(
    AdvancedSettings=(RotationOrderForFilter=XZY)
    Child=(Type=Bone,Name="deform_button_bottom")
    Parents=((Item=(Type=Bone,Name="spine_01"),Weight=0.0284),(Item=(Type=Bone,Name="pelvis"),Weight=0.0372),(Item=(Type=Bone,Name="spine_03"),Weight=0.1247),(Item=(Type=Bone,Name="spine_02"),Weight=0.2406),(Item=(Type=Bone,Name="spine_04"),Weight=0.5692))
    Weight=1
   )
   > &RigUnit_ParentConstraint(
    AdvancedSettings=(RotationOrderForFilter=XZY)
    Child=(Type=Bone,Name="deform_button_top")
    Parents=((Item=(Type=Bone,Name="spine_03"),Weight=0.0054),(Item=(Type=Bone,Name="deform_pec_l"),Weight=0.0223),(Item=(Type=Bone,Name="spine_05"),Weight=0.1196),(Item=(Type=Bone,Name="spine_04"),Weight=0.8528))
    Weight=1
   )
   > &RigUnit_ParentConstraint(
    AdvancedSettings=(RotationOrderForFilter=XZY)
    Child=(Type=Bone,Name="dyn_lace_l")
    Parents=((Item=(Type=Bone,Name="calf_twist_01_l"),Weight=0.3076),(Item=(Type=Bone,Name="foot_l"),Weight=0.6924))
    Weight=1
   )
   > &RigUnit_ParentConstraint(
    AdvancedSettings=(RotationOrderForFilter=XZY)
    Child=(Type=Bone,Name="dyn_lace_r")
    Parents=((Item=(Type=Bone,Name="calf_twist_01_r"),Weight=0.3076),(Item=(Type=Bone,Name="foot_r"),Weight=0.6924))
    Weight=1
   )
   RigVMFunction_ControlFlowBranch = &RigVMFunction_ControlFlowBranch(Condition=RigVMVariableNode() {variable=Disable Floor Slope}.Value, ExecuteContext=@RigVMFunction_Sequence.G)
   KalmanFilter = &RigVMFunction_KalmanFloat(BufferSize=16, Value=RigVMVariableNode() {variable=Push Down Weight}.Value) {template="KalmanFilter::Execute(in Value,in BufferSize,out Result)"}
   &RigUnit_ModifyTransforms(
    ItemToModify=((Item=(Type=Bone,Name="dyn_skirt_ground_collision"),Transform=(Rotation=(X=0,Y=0,Z=0,W=1),Translation=(X=0,Y=0,Z=-500),Scale3D=(X=1,Y=1,Z=1))))
    WeightMinimum=0
    Weight=@KalmanFilter.Result
    ExecutePin=@RigVMFunction_ControlFlowBranch.Completed
   )
   RigVMVariableNode(Value=0, ExecuteContext=@RigVMFunction_ControlFlowBranch.False) {variable=Push Down Weight}
   RigVMVariableNode(Value=1, ExecuteContext=@RigVMFunction_ControlFlowBranch.True) {variable=Push Down Weight}
   n33 = &RigVMFunction_ControlFlowBranch(Condition=RigVMVariableNode() {variable=isSliding}.Value, ExecuteContext=@RigVMFunction_Sequence.H)
   AlphaInterp = &RigVMFunction_AlphaInterp(InterpSpeedDecreasing=5, InterpSpeedIncreasing=5, Scale=1, bClampResult=False, Value=RigVMVariableNode() {variable=slideAlpha}.Value)
   &RigUnit_ModifyTransforms(
    ItemToModify="((Item=(Type=Bone,Name=\"dyn_skirt_back_l_MSTR\"),Transform=(Rotation=(X=-0.2917,Y=-0.5861,Z=-0.2024,W=0.7283),Translation=(X=-5.6678,Y=3.2131,Z=0.3603),Scale3D=(X=1,Y=1,Z=1))),(Item=(Type=Bone,Name=\"dyn_skirt_back_mid_MSTR\"),Transform=(Rotation=(X=0,Y=-0.73 ...(truncated)"
    WeightMinimum=0
    Weight=@AlphaInterp.Result
    ExecutePin=@n33.Completed
   )
   RigVMVariableNode(Value=0, ExecuteContext=@n33.False) {variable=slideAlpha}
   RigVMVariableNode(Value=1, ExecuteContext=@n33.True) {variable=slideAlpha}
   &RigVMFunction_AlphaInterp(InterpSpeedDecreasing=10, InterpSpeedIncreasing=10, Scale=1, bClampResult=False, bInterpResult=True, Value=@n28.OutputParam)
  ]}
 RigVMModel Construction Event Graph {nodes [
   n1 = &RigUnit_PrepareForExecution()
   n2 = &RigUnit_GetTransform(Item=(Type=Bone,Name="thigh_r"), Space=LocalSpace, bInitial=False)
   &RigUnit_SetRotation(
    Item=(Type=Bone,Name="CR_thigh_r")
    Weight=1
    bInitial=true
    bPropagateToChildren=false
    ExecutePin=@n1.ExecutePin
    Value=@n2.Transform.Rotation
   )
   > &RigUnit_SetRotation(Item=(Type=Bone,Name="CR_thigh_r_no_twist"), Value=(X=0,Y=0,Z=0,W=1), Weight=1, bInitial=true, bPropagateToChildren=false)
   &RigVMFunction_MathQuaternionInverse(Value=&RigVMFunction_MathQuaternionSwingTwist(TwistAxis=(X=-1,Y=0,Z=0), Input=@n2.Transform.Rotation).Twist) {template="Inverse::Execute(in Value,out Result)"}
  ]}
]
hierarchy [
 root {loc=<0 0 0>, rot=<0 0 0 1>}
 pelvis {parent=root, loc=<0 -0.5301 98.0571>, rot=<0.0127 0.707 -0.0127 -0.707>}
 spine_01 {parent=pelvis, loc=<6.1208 -0.9328 0>, rot=<0 0 0.034 -0.9994>}
 spine_02 {parent=spine_01, loc=<6.5997 -0.2095 0>, rot=<0 0 0.0238 -0.9997>}
 spine_03 {parent=spine_02, loc=<6.4914 -0.1237 0>, rot=<0 0 0.0149 -0.9999>}
 spine_04 {parent=spine_03, loc=<6.9598 0.0794 0>, rot=<0 0 -0.0132 -0.9999>}
 spine_05 {parent=spine_04, loc=<6.9987 0 0>, rot=<0 0 -0.0015 -1>}
 neck_01 {parent=spine_05, loc=<11.3138 0.1529 0>, rot=<0 0 -0.2003 -0.9797>}
 neck_02 {parent=neck_01, loc=<4.5627 0 0>, rot=<0 0 0.0504 -0.9987>}
 head {parent=neck_02, loc=<5.3914 0 0>, rot=<0 0 0.115 -0.9934>}
 clavicle_l {parent=spine_05, loc=<5.4831 -0.1053 -3.2946>, rot=<0.0465 -0.7289 -0.009 -0.6829>}
 upperarm_l {parent=clavicle_l, loc=<9.8658 -1.2136 -0.4105>, rot=<0.0164 -0.3245 -0.0269 -0.9454>}
 lowerarm_l {parent=upperarm_l, loc=<26.465 -0.8345 0.259>, rot=<0.0036 0.0161 -0.2298 -0.9731>}
 hand_l {parent=lowerarm_l, loc=<22.7355 -0.3203 -0.6105>, rot=<-0.6784 -0.0223 0.0735 -0.7307>}
 thumb_01_l {parent=hand_l, loc=<1.3118 -0.3191 -2.459>, rot=<0.6546 -0.4042 0.0664 -0.6354>}
 thumb_02_l {parent=thumb_01_l, loc=<5.9336 -0.0448 0.0401>, rot=<-0.0037 0.0213 0.298 -0.9543>}
 thumb_03_l {parent=thumb_02_l, loc=<2.2223 0.3682 -0.0907>, rot=<0.0118 -0.013 0.0066 -0.9998>}
 index_metacarpal_l {parent=hand_l, loc=<1.6535 0.0183 -2.96>, rot=<0.0158 -0.0124 -0.0173 -0.9996>}
 index_01_l {parent=index_metacarpal_l, loc=<8.3008 0.0357 -0.0524>, rot=<-0.0044 0.0485 0.0924 -0.9945>}
 index_02_l {parent=index_01_l, loc=<3.3587 -0.3004 0.0452>, rot=<-0.002 0.0027 0.3225 -0.9466>}
 index_03_l {parent=index_02_l, loc=<2.5578 -0.123 -0.0252>, rot=<-0.002 0.0067 0.2558 -0.9667>}
 middle_metacarpal_l {parent=hand_l, loc=<2.1856 -0.1725 -0.5984>, rot=<0.006 0.0168 -0.0329 -0.9993>}
 middle_01_l {parent=middle_metacarpal_l, loc=<7.3275 -0.2447 0.0129>, rot=<-0.0208 0.0042 0.0858 -0.9961>}
 middle_02_l {parent=middle_01_l, loc=<4.0123 0.1401 0.1175>, rot=<-0.0218 0.0258 0.3512 -0.9357>}
 middle_03_l {parent=middle_02_l, loc=<2.6077 0.0534 0.1373>, rot=<0.0082 0.0188 0.1708 -0.9851>}
 ring_metacarpal_l {parent=hand_l, loc=<2.2074 -0.3625 1.3638>, rot=<-0.0284 0.0722 -0.0458 -0.9959>}
 ring_01_l {parent=ring_metacarpal_l, loc=<7.0295 -0.6197 -0.1446>, rot=<-0.0489 -0.0252 0.1892 -0.9804>}
 ring_02_l {parent=ring_01_l, loc=<3.6174 0.4647 -0.1027>, rot=<0 0 0.2806 -0.9598>}
 ring_03_l {parent=ring_02_l, loc=<2.4152 0.114 0.0199>, rot=<0 0 0.1189 -0.9929>}
 pinky_metacarpal_l {parent=hand_l, loc=<1.7616 -0.9753 3.3814>, rot=<-0.0397 0.0814 -0.069 -0.9935>}
 pinky_01_l {parent=pinky_metacarpal_l, loc=<7.1709 -0.4831 -0.1094>, rot=<-0.0654 -0.0035 0.2242 -0.9723>}
 pinky_02_l {parent=pinky_01_l, loc=<2.8274 0.0893 -0.1043>, rot=<0.0037 0.0011 0.2854 -0.9584>}
 pinky_03_l {parent=pinky_02_l, loc=<2.2046 0.0307 -0.0078>, rot=<-0.0002 0.0098 0.1445 -0.9895>}
 deform_wrist_dn_base_l {parent=hand_l, loc=<0 -1.5326 0>, rot=<0 0 0 -1>}
 deform_wrist_dn_l {parent=deform_wrist_dn_base_l, loc=<0 0 0>, rot=<0 0 0 1>}
 deform_wrist_up_base_l {parent=hand_l, loc=<0 2.3009 0>, rot=<0 0 0 -1>}
 deform_wrist_up_l {parent=deform_wrist_up_base_l, loc=<0 0 0>, rot=<0 0 0 1>}
 lowerarm_twist_02_l {parent=lowerarm_l, loc=<10.6707 0 0>, rot=<0 0 0 1>}
 lowerarm_twist_01_l {parent=lowerarm_l, loc=<18.316 0 0>, rot=<0 0 0 1>}
 deform_elbow_in_base_l {parent=lowerarm_l, loc=<0.1834 3.6062 -0.2572>, rot=<0 0 0 -1>}
 deform_elbow_in_l {parent=deform_elbow_in_base_l, loc=<0 0 0>, rot=<0 0 0 1>}
 deform_elbow_out_base_l {parent=lowerarm_l, loc=<0 -4 0>, rot=<0 0 0 1>}
 deform_elbow_out_l {parent=deform_elbow_out_base_l, loc=<0 0 0>, rot=<0 0 0 1>}
 upperarm_twist_01_l {parent=upperarm_l, loc=<7.2135 -0.1134 0>, rot=<0 0 0 1>}
 deform_upperarm_bk_base_l {parent=upperarm_twist_01_l, loc=<-6.4053 -4.3746 0.314>, rot=<0 0 0 -1>}
 deform_upperarm_bk_l {parent=deform_upperarm_bk_base_l, loc=<0 0 0>, rot=<0 0 0 1>}
 deform_upperarm_fr_base_l {parent=upperarm_twist_01_l, loc=<-5.6312 5.092 0.3354>, rot=<0 0 0 -1>}
 deform_upperarm_fr_l {parent=deform_upperarm_fr_base_l, loc=<0 0 0>, rot=<0 0 0 1>}
 deform_upperarm_up_base_l {parent=upperarm_twist_01_l, loc=<-8.891 0.4825 3.9517>, rot=<0 0 0 -1>}
 deform_upperarm_up_l {parent=deform_upperarm_up_base_l, loc=<0 0 0>, rot=<0 0 0 1>}
 upperarm_twist_02_l {parent=upperarm_l, loc=<19.7998 -0.1134 0>, rot=<0 0 0 1>}
 clavicle_r {parent=spine_05, loc=<5.4831 -0.1053 3.2946>, rot=<0.7289 0.0465 -0.6829 0.009>}
 upperarm_r {parent=clavicle_r, loc=<-9.8658 1.2136 0.4105>, rot=<0.0164 -0.3245 -0.0269 -0.9454>}
 lowerarm_r {parent=upperarm_r, loc=<-26.465 0.8345 -0.259>, rot=<0.0036 0.0161 -0.2298 -0.9731>}
 hand_r {parent=lowerarm_r, loc=<-22.7355 0.3203 0.6106>, rot=<-0.6784 -0.0223 0.0735 -0.7307>}
 thumb_01_r {parent=hand_r, loc=<-1.3118 0.3191 2.459>, rot=<0.6546 -0.4042 0.0664 -0.6354>}
 thumb_02_r {parent=thumb_01_r, loc=<-5.9336 0.0448 -0.0401>, rot=<-0.0037 0.0213 0.298 -0.9543>}
 thumb_03_r {parent=thumb_02_r, loc=<-2.2223 -0.3682 0.0907>, rot=<0.0118 -0.013 0.0066 -0.9998>}
 index_metacarpal_r {parent=hand_r, loc=<-1.6535 -0.0183 2.96>, rot=<0.0158 -0.0124 -0.0173 -0.9996>}
 index_01_r {parent=index_metacarpal_r, loc=<-8.3008 -0.0357 0.0524>, rot=<-0.0044 0.0485 0.0924 -0.9945>}
 index_02_r {parent=index_01_r, loc=<-3.3587 0.3004 -0.0452>, rot=<-0.002 0.0027 0.3225 -0.9466>}
 index_03_r {parent=index_02_r, loc=<-2.5578 0.123 0.0252>, rot=<-0.002 0.0067 0.2558 -0.9667>}
 middle_metacarpal_r {parent=hand_r, loc=<-2.1856 0.1725 0.5984>, rot=<0.006 0.0168 -0.0329 -0.9993>}
 middle_01_r {parent=middle_metacarpal_r, loc=<-7.3275 0.2447 -0.0129>, rot=<-0.0208 0.0042 0.0858 -0.9961>}
 middle_02_r {parent=middle_01_r, loc=<-4.0123 -0.1401 -0.1175>, rot=<-0.0218 0.0258 0.3512 -0.9357>}
 middle_03_r {parent=middle_02_r, loc=<-2.6077 -0.0534 -0.1373>, rot=<0.0082 0.0188 0.1708 -0.9851>}
 ring_metacarpal_r {parent=hand_r, loc=<-2.2074 0.3625 -1.3638>, rot=<-0.0284 0.0722 -0.0458 -0.9959>}
 ring_01_r {parent=ring_metacarpal_r, loc=<-7.0295 0.6197 0.1446>, rot=<-0.0489 -0.0252 0.1892 -0.9804>}
 ring_02_r {parent=ring_01_r, loc=<-3.6174 -0.4647 0.1027>, rot=<0 0 0.2806 -0.9598>}
 ring_03_r {parent=ring_02_r, loc=<-2.4152 -0.114 -0.0199>, rot=<0 0 0.1189 -0.9929>}
 pinky_metacarpal_r {parent=hand_r, loc=<-1.7616 0.9753 -3.3814>, rot=<-0.0397 0.0814 -0.069 -0.9935>}
 pinky_01_r {parent=pinky_metacarpal_r, loc=<-7.1709 0.4831 0.1094>, rot=<-0.0654 -0.0035 0.2242 -0.9723>}
 pinky_02_r {parent=pinky_01_r, loc=<-2.8274 -0.0893 0.1043>, rot=<0.0037 0.0011 0.2854 -0.9584>}
 pinky_03_r {parent=pinky_02_r, loc=<-2.2046 -0.0307 0.0078>, rot=<-0.0002 0.0098 0.1445 -0.9895>}
 deform_wrist_dn_base_r {parent=hand_r, loc=<-0.0001 1.5328 0>, rot=<0 0 0 -1>}
 deform_wrist_dn_r {parent=deform_wrist_dn_base_r, loc=<0 0 0>, rot=<0 0 0 1>}
 deform_wrist_up_base_r {parent=hand_r, loc=<-0.0002 -2.3007 -0.0001>, rot=<0 0 0 -1>}
 deform_wrist_up_r {parent=deform_wrist_up_base_r, loc=<0 0 0>, rot=<0 0 0 1>}
 lowerarm_twist_02_r {parent=lowerarm_r, loc=<-10.6707 0 0>, rot=<0 0 0 -1>}
 lowerarm_twist_01_r {parent=lowerarm_r, loc=<-18.316 0 0>, rot=<0 0 0 -1>}
 deform_elbow_in_base_r {parent=lowerarm_r, loc=<-0.1834 -3.6062 0.2572>, rot=<0 0 0 -1>}
 deform_elbow_in_r {parent=deform_elbow_in_base_r, loc=<0 0 0>, rot=<0 0 0 1>}
 deform_elbow_out_base_r {parent=lowerarm_r, loc=<0 4 -0.0001>, rot=<0 0 0 -1>}
 deform_elbow_out_r {parent=deform_elbow_out_base_r, loc=<0 0 0>, rot=<0 0 0 1>}
 upperarm_twist_01_r {parent=upperarm_r, loc=<-7.2135 0.1134 0>, rot=<0 0 0 -1>}
 deform_upperarm_bk_base_r {parent=upperarm_twist_01_r, loc=<6.4053 4.3746 -0.314>, rot=<0 0 0 -1>}
 deform_upperarm_bk_r {parent=deform_upperarm_bk_base_r, loc=<0 0 0>, rot=<0 0 0 1>}
 deform_upperarm_fr_base_r {parent=upperarm_twist_01_r, loc=<5.6312 -5.092 -0.3354>, rot=<0 0 0 -1>}
 deform_upperarm_fr_r {parent=deform_upperarm_fr_base_r, loc=<0 0 0>, rot=<0 0 0 1>}
 deform_upperarm_up_base_r {parent=upperarm_twist_01_r, loc=<8.891 -0.4825 -3.9517>, rot=<0 0 0 -1>}
 deform_upperarm_up_r {parent=deform_upperarm_up_base_r, loc=<0 0 0>, rot=<0 0 0 1>}
 upperarm_twist_02_r {parent=upperarm_r, loc=<-19.7998 0.1134 0>, rot=<0 0 0 -1>}
 deform_thing_B_wrist {parent=upperarm_r, loc=<-11.2808 4.0918 -16.1775>, rot=<0.6343 -0.6708 0.2687 0.2747>}
 deform_thing_B_hand {parent=deform_thing_B_wrist, loc=<0 0 0>, rot=<0 0 0 1>}
 deform_thing_B_thumb_01 {parent=deform_thing_B_hand, loc=<2.6885 -3.5662 -1.5723>, rot=<-0.1157 -0.3522 0.2349 -0.8986>}
 deform_thing_B_thumb_02 {parent=deform_thing_B_thumb_01, loc=<3.4825 0 0>, rot=<0 0 -0.206 -0.9786>}
 deform_thing_B_thumb_03 {parent=deform_thing_B_thumb_02, loc=<2.8996 0 0>, rot=<0 0 -0.0406 -0.9992>}
 deform_thing_B_index_metacarpal {parent=deform_thing_B_hand, loc=<1.9918 -3.3183 -0.6339>, rot=<0.6552 0.0388 0.0117 -0.7543>}
 deform_thing_B_index_01 {parent=deform_thing_B_index_metacarpal, loc=<7.4719 0 0>, rot=<0 0 -0.1468 -0.9892>}
 deform_thing_B_index_02 {parent=deform_thing_B_index_01, loc=<4.1266 0 0>, rot=<0 0 -0.3324 -0.9431>}
 deform_thing_B_index_03 {parent=deform_thing_B_index_02, loc=<2.6865 0 0>, rot=<0 0 -0.1779 -0.9841>}
 deform_thing_B_middle_metacarpal {parent=deform_thing_B_hand, loc=<2.2933 -1.6479 0.2632>, rot=<0.6843 0.0527 -0.0463 -0.7258>}
 deform_thing_B_middle_01 {parent=deform_thing_B_middle_metacarpal, loc=<7.0419 0 0>, rot=<0 0 -0.1092 -0.994>}
 deform_thing_B_middle_02 {parent=deform_thing_B_middle_01, loc=<4.5892 0 0>, rot=<0 0 -0.2986 -0.9544>}
 deform_thing_B_middle_03 {parent=deform_thing_B_middle_02, loc=<2.6375 0 0>, rot=<0 0 -0.1978 -0.9802>}
 deform_thing_B_ring_metacarpal {parent=deform_thing_B_hand, loc=<2.2496 0.9403 0.4665>, rot=<0.7384 0.0457 -0.0664 -0.6695>}
 deform_thing_B_ring_01 {parent=deform_thing_B_ring_metacarpal, loc=<6.7033 0 0>, rot=<0 0 -0.1416 -0.9899>}
 deform_thing_B_ring_02 {parent=deform_thing_B_ring_01, loc=<4.1845 0 0>, rot=<0 0 -0.2175 -0.9761>}
 deform_thing_B_ring_03 {parent=deform_thing_B_ring_02, loc=<2.3557 0 0>, rot=<0 0 -0.1412 -0.99>}
 deform_thing_B_pinky_metacarpal {parent=deform_thing_B_hand, loc=<1.9337 2.6862 -0.1754>, rot=<0.7676 0.0649 -0.0896 -0.6313>}
 deform_thing_B_pinky_01 {parent=deform_thing_B_pinky_metacarpal, loc=<6.6008 0 0>, rot=<0 0 -0.0962 -0.9954>}
 deform_thing_B_pinky_02 {parent=deform_thing_B_pinky_01, loc=<3.5766 0 0>, rot=<0 0 -0.3162 -0.9487>}
 deform_thing_B_pinky_03 {parent=deform_thing_B_pinky_02, loc=<1.9766 0 0>, rot=<0 0 0.0262 -0.9997>}
 deform_thing_A_wrist {parent=clavicle_r, loc=<-14.1532 3.5716 -15.59>, rot=<0.6664 -0.6208 0.1611 0.3801>}
 deform_thing_A_hand {parent=deform_thing_A_wrist, loc=<0 0 0>, rot=<0 0 0 1>}
 deform_thing_A_thumb_01 {parent=deform_thing_A_hand, loc=<2.6885 -3.5662 -1.5723>, rot=<-0.1157 -0.3522 0.2349 -0.8986>}
 deform_thing_A_thumb_02 {parent=deform_thing_A_thumb_01, loc=<3.4825 0 0>, rot=<0 0 -0.206 -0.9786>}
 deform_thing_A_thumb_03 {parent=deform_thing_A_thumb_02, loc=<2.8996 0 0>, rot=<0 0 -0.0668 -0.9978>}
 deform_thing_A_index_metacarpal {parent=deform_thing_A_hand, loc=<1.9918 -3.3183 -0.6339>, rot=<0.6552 0.0388 0.0117 -0.7543>}
 deform_thing_A_index_01 {parent=deform_thing_A_index_metacarpal, loc=<7.4719 0 0>, rot=<0 0 -0.1468 -0.9892>}
 deform_thing_A_index_02 {parent=deform_thing_A_index_01, loc=<4.1266 0 0>, rot=<0 0 -0.3324 -0.9431>}
 deform_thing_A_index_03 {parent=deform_thing_A_index_02, loc=<2.6864 0 0>, rot=<0 0 -0.2036 -0.9791>}
 deform_thing_A_middle_metacarpal {parent=deform_thing_A_hand, loc=<2.2933 -1.6479 0.2632>, rot=<0.6843 0.0527 -0.0463 -0.7258>}
 deform_thing_A_middle_01 {parent=deform_thing_A_middle_metacarpal, loc=<7.0419 0 0>, rot=<0 0 -0.1092 -0.994>}
 deform_thing_A_middle_02 {parent=deform_thing_A_middle_01, loc=<4.5892 0 0>, rot=<0 0 -0.2986 -0.9544>}
 deform_thing_A_middle_03 {parent=deform_thing_A_middle_02, loc=<2.6375 0 0>, rot=<0 0 -0.2234 -0.9747>}
 deform_thing_A_ring_metacarpal {parent=deform_thing_A_hand, loc=<2.2496 0.9403 0.4665>, rot=<0.7384 0.0457 -0.0664 -0.6695>}
 deform_thing_A_ring_01 {parent=deform_thing_A_ring_metacarpal, loc=<6.7033 0 0>, rot=<0 0 -0.1416 -0.9899>}
 deform_thing_A_ring_02 {parent=deform_thing_A_ring_01, loc=<4.1845 0 0>, rot=<0 0 -0.2175 -0.9761>}
 deform_thing_A_ring_03 {parent=deform_thing_A_ring_02, loc=<2.3557 0 0>, rot=<0 0 -0.1671 -0.9859>}
 deform_thing_A_pinky_metacarpal {parent=deform_thing_A_hand, loc=<1.9337 2.6862 -0.1754>, rot=<0.7676 0.0649 -0.0896 -0.6313>}
 deform_thing_A_pinky_01 {parent=deform_thing_A_pinky_metacarpal, loc=<6.6008 0 0>, rot=<0 0 -0.0962 -0.9954>}
 deform_thing_A_pinky_02 {parent=deform_thing_A_pinky_01, loc=<3.5766 0 0>, rot=<0 0 -0.3162 -0.9487>}
 deform_thing_A_pinky_03 {parent=deform_thing_A_pinky_02, loc=<1.9766 0 0>, rot=<0 0 0 -1>}
 attach_backpack {parent=spine_05, loc=<-0.1237 -13.4688 -0.0001>, rot=<-0.0093 -0.7071 0.0093 -0.707>}
 Backpack_BR {parent=attach_backpack, type=Null, loc=<0 4.6912 -6.6107>, rot=<0.0058 0 0 1>}
 attach_cape {parent=spine_05, loc=<9.0149 -11.7101 0.0002>, rot=<0.0616 -0.7044 -0.0616 -0.7044>}
 cape {parent=attach_cape, type=Null, loc=<0 2.875 -2.8854>, rot=<0 0 0 1>}
 deform_pec_aim_base_l {parent=spine_04, loc=<3.2456 9.3309 -1.2237>, rot=<0.0603 -0.4033 0.3318 -0.8507>}
 deform_pec_aim_l {parent=deform_pec_aim_base_l, loc=<0 0 0>, rot=<0 0 0 1>}
 deform_pec_base_l {parent=deform_pec_aim_l, loc=<9.3665 0 0>, rot=<0 0 0 1>}
 deform_pec_l {parent=deform_pec_base_l, loc=<0 0 0>, rot=<0 0 0 1>}
 deform_pec_aim_base_r {parent=spine_04, loc=<3.2452 9.3309 1.2237>, rot=<-0.4033 -0.0603 0.8507 0.3318>}
 deform_pec_aim_r {parent=deform_pec_aim_base_r, loc=<0 0 0>, rot=<0 0 0 1>}
 deform_pec_base_r {parent=deform_pec_aim_r, loc=<-9.3667 -0.0003 -0.0004>, rot=<0 0 0 1>}
 deform_pec_r {parent=deform_pec_base_r, loc=<0 0 0>, rot=<0 0 0 1>}
 deform_button_top {parent=spine_03, loc=<6.9833 11.7117 -1.0235>, rot=<0.4942 0.5057 0.5057 0.4942>}
 deform_button_bottom {parent=spine_02, loc=<5.7186 10.5784 -0.9677>, rot=<0.4454 0.5492 0.5492 0.4454>}
 thigh_l {parent=pelvis, loc=<-1.0855 -0.2109 -8.7053>, rot=<0.0009 0.0355 0.0155 -0.9993>}
 calf_l {parent=thigh_l, loc=<-42.5284 -0.0019 -0.7297>, rot=<0.0006 -0.0161 -0.0391 -0.9991>}
 foot_l {parent=calf_l, loc=<-46.7401 -2.5014 -0.592>, rot=<0.0427 -0.0238 0.0353 -0.9982>}
 ball_l {parent=foot_l, loc=<-3.5658 10.8066 -0.1544>, rot=<0.0001 -0.0001 -0.7063 -0.7079>}
 dyn_lace_l {parent=foot_l, loc=<4.0413 6.6058 0.544>, rot=<0.2736 0.6349 0.6521 0.3111>}
 dyn_lace_l_end_l_MSTR {parent=dyn_lace_l, loc=<0.0857 -0.6136 0.0525>, rot=<-0.2643 -0.0774 0.2931 -0.9156>}
 dyn_lace_l_end_l_01 {parent=dyn_lace_l_end_l_MSTR, loc=<0 0 0>, rot=<0 0 0 1>}
 dyn_lace_l_end_l_02 {parent=dyn_lace_l_end_l_01, loc=<3.3423 0 0>, rot=<-0.0728 -0.0515 0.0403 -0.9952>}
 dyn_lace_l_loop_l_MSTR {parent=dyn_lace_l, loc=<-0.2431 -0.544 0.0521>, rot=<-0.1898 -0.2286 0.4668 -0.8329>}
 dyn_lace_l_loop_l_01 {parent=dyn_lace_l_loop_l_MSTR, loc=<0 0 0>, rot=<0 0 0 1>}
 dyn_lace_l_loop_l_02 {parent=dyn_lace_l_loop_l_01, loc=<2.4346 0 0>, rot=<-0.1737 -0.1815 0.0253 -0.9676>}
 dyn_lace_l_end_r_MSTR {parent=dyn_lace_l, loc=<0.4407 0.3181 0.1808>, rot=<0.1938 -0.1182 -0.3496 -0.909>}
 dyn_lace_l_end_r_01 {parent=dyn_lace_l_end_r_MSTR, loc=<0 0 0>, rot=<0 0 0 1>}
 dyn_lace_l_end_r_02 {parent=dyn_lace_l_end_r_01, loc=<2.6774 0 0>, rot=<0.0168 -0.0579 0.0323 -0.9977>}
 dyn_lace_l_loop_r_MSTR {parent=dyn_lace_l, loc=<0.0971 0.4939 -0.0208>, rot=<0.1866 -0.2284 -0.649 -0.7013>}
 dyn_lace_l_loop_r_01 {parent=dyn_lace_l_loop_r_MSTR, loc=<0 0 0>, rot=<0 0 0 1>}
 dyn_lace_l_loop_r_02 {parent=dyn_lace_l_loop_r_01, loc=<2.0373 0 0>, rot=<0.0688 -0.2458 0.046 -0.9658>}
 calf_twist_02_l {parent=calf_l, loc=<-19.1546 -1.7943 -0.2638>, rot=<0.0237 0.0095 -0.0035 -0.9997>}
 calf_twist_01_l {parent=calf_l, loc=<-35.8483 -1.9197 -0.5794>, rot=<0.0237 0.0095 -0.0035 -0.9997>}
 deform_knee_in_base_l {parent=calf_l, loc=<0.8031 -5.0231 0.2456>, rot=<0 0 0 -1>}
 deform_knee_in_l {parent=deform_knee_in_base_l, loc=<0 0 0>, rot=<0 0 0 1>}
 deform_knee_out_base_l {parent=calf_l, loc=<0 4 0>, rot=<0 0 0 -1>}
 deform_knee_out_l {parent=deform_knee_out_base_l, loc=<0 0 0>, rot=<0 0 0 1>}
 thigh_twist_01_l {parent=thigh_l, loc=<-27.952 0 0>, rot=<0.0013 0 0 -1>}
 deform_thigh_front_l {parent=thigh_twist_01_l, loc=<0.0577 6.6747 -1.134>, rot=<-0.4771 -0.5537 -0.4349 -0.526>}
 deform_groin_base_l {parent=thigh_l, loc=<0.4801 5.6776 -0.2486>, rot=<0 0 0 -1>}
 deform_groin_l {parent=deform_groin_base_l, loc=<0 0 0>, rot=<0 0 0 1>}
 CR_thigh_l_no_twist {parent=thigh_l, loc=<0 0 0>, rot=<0 0 0 1>}
 thigh_r {parent=pelvis, loc=<-1.0855 -0.2109 8.7053>, rot=<0.0355 -0.0009 0.9993 0.0155>}
 calf_r {parent=thigh_r, loc=<42.5284 0.0019 0.7297>, rot=<0.0006 -0.0161 -0.0391 -0.9991>}
 foot_r {parent=calf_r, loc=<46.7401 2.5014 0.592>, rot=<0.0427 -0.0238 0.0353 -0.9982>}
 ball_r {parent=foot_r, loc=<3.5658 -10.8066 0.1544>, rot=<0.0001 -0.0001 -0.7063 -0.7079>}
 dyn_lace_r {parent=foot_r, loc=<-4.0413 -6.6058 -0.544>, rot=<-0.2657 0.6306 0.6889 -0.2391>}
 dyn_lace_r_end_l_MSTR {parent=dyn_lace_r, loc=<0.1914 -0.2675 -0.4689>, rot=<0.0618 -0.7255 0.3732 -0.5749>}
 dyn_lace_r_end_l_01 {parent=dyn_lace_r_end_l_MSTR, loc=<0 0 0>, rot=<0 0 0 1>}
 dyn_lace_r_end_l_02 {parent=dyn_lace_r_end_l_01, loc=<2.6774 0 0>, rot=<0.018 -0.0595 -0.0292 -0.9976>}
 dyn_lace_r_loop_l_MSTR {parent=dyn_lace_r, loc=<-0.0096 -0.4816 -0.1474>, rot=<0.3224 -0.696 0.5365 -0.3518>}
 dyn_lace_r_loop_l_01 {parent=dyn_lace_r_loop_l_MSTR, loc=<0 0 0>, rot=<0 0 0 1>}
 dyn_lace_r_loop_l_02 {parent=dyn_lace_r_loop_l_01, loc=<2.0373 0 0>, rot=<-0.1527 -0.2472 -0.0374 -0.9561>}
 dyn_lace_r_loop_r_MSTR {parent=dyn_lace_r, loc=<-0.0872 0.4256 -0.0274>, rot=<-0.3151 -0.6918 -0.4981 -0.4171>}
 dyn_lace_r_loop_r_01 {parent=dyn_lace_r_loop_r_MSTR, loc=<0 0 0>, rot=<0 0 0 1>}
 dyn_lace_r_loop_r_02 {parent=dyn_lace_r_loop_r_01, loc=<1.8111 0 0>, rot=<0.174 -0.2631 0.0024 -0.9489>}
 dyn_lace_r_end_r_MSTR {parent=dyn_lace_r, loc=<-0.0346 0.6144 0.0088>, rot=<-0.0827 -0.7014 -0.3825 -0.5957>}
 dyn_lace_r_end_r_01 {parent=dyn_lace_r_end_r_MSTR, loc=<0 0 0>, rot=<0 0 0 1>}
 dyn_lace_r_end_r_02 {parent=dyn_lace_r_end_r_01, loc=<3.3474 0 0>, rot=<0.0648 -0.067 -0.039 -0.9949>}
 calf_twist_02_r {parent=calf_r, loc=<19.1546 1.7943 0.2638>, rot=<0.0237 0.0095 -0.0035 -0.9997>}
 calf_twist_01_r {parent=calf_r, loc=<35.8483 1.9197 0.5794>, rot=<0.0237 0.0095 -0.0035 -0.9997>}
 deform_knee_in_base_r {parent=calf_r, loc=<-0.8031 5.0231 -0.2457>, rot=<0 0 0 -1>}
 deform_knee_in_r {parent=deform_knee_in_base_r, loc=<0 0 0>, rot=<0 0 0 1>}
 deform_knee_out_base_r {parent=calf_r, loc=<0 -4 0>, rot=<0 0 0 -1>}
 deform_knee_out_r {parent=deform_knee_out_base_r, loc=<0 0 0>, rot=<0 0 0 1>}
 thigh_twist_01_r {parent=thigh_r, loc=<27.952 0 0>, rot=<0.0013 0 0 -1>}
 deform_thigh_front_r {parent=thigh_twist_01_r, loc=<-0.0577 -6.6747 1.134>, rot=<-0.4349 0.526 0.4771 -0.5537>}
 deform_groin_base_r {parent=thigh_r, loc=<-0.4801 -5.6776 0.2486>, rot=<0 0 0 -1>}
 deform_groin_r {parent=deform_groin_base_r, loc=<0 0 0>, rot=<0 0 0 1>}
 CR_thigh_r_no_twist {parent=thigh_r, loc=<0 0 0>, rot=<0 0 0 1>}
 CR_thigh_r {parent=thigh_r, loc=<0 0 0>, rot=<0.0355 -0.0009 0.9993 0.0155>}
 deform_glute_base_l {parent=pelvis, loc=<-7.3466 -9.0213 -9.1245>, rot=<0.0172 0.031 -0.4477 -0.8935>}
 deform_glute_l {parent=deform_glute_base_l, loc=<0 0 0>, rot=<0 0 0 1>}
 deform_glute_base_r {parent=pelvis, loc=<-7.3464 -9.0213 9.1245>, rot=<0.031 -0.0172 0.8935 -0.4477>}
 deform_glute_r {parent=deform_glute_base_r, loc=<0 0 0>, rot=<0 0 0 1>}
 dyn_skirt_front_l_MSTR {parent=pelvis, loc=<-4.9056 7.2664 -10.6819>, rot=<0.0068 -0.8018 -0.596 -0.0429>}
 dyn_skirt_front_l_01 {parent=dyn_skirt_front_l_MSTR, loc=<0 0 0>, rot=<0 0 0 1>}
 dyn_skirt_front_l_02 {parent=dyn_skirt_front_l_01, loc=<18.1754 0 0>, rot=<0 -0.0008 -0.0023 -1>}
 dyn_skirt_front_l_03 {parent=dyn_skirt_front_l_02, loc=<20.474 -0.0692 -0.1901>, rot=<0 -0.0004 -0.0011 -1>}
 dyn_skirt_front_l_04 {parent=dyn_skirt_front_l_03, loc=<18.1738 -0.0647 -0.1778>, rot=<0 0 0 -1>}
 dyn_skirt_front_r_MSTR {parent=pelvis, loc=<-4.9699 7.0015 10.4817>, rot=<0.053 0.5243 0.8499 0.0008>}
 dyn_skirt_front_r_01 {parent=dyn_skirt_front_r_MSTR, loc=<0 0 0>, rot=<0 0 0 1>}
 dyn_skirt_front_r_02 {parent=dyn_skirt_front_r_01, loc=<18.1425 0 0>, rot=<0 -0.0055 0.0066 -1>}
 dyn_skirt_front_r_03 {parent=dyn_skirt_front_r_02, loc=<20.4604 0.323 -0.3851>, rot=<0 -0.0035 0.0041 -1>}
 dyn_skirt_front_r_04 {parent=dyn_skirt_front_r_03, loc=<18.2772 0.4696 -0.5599>, rot=<0 0 0 -1>}
 dyn_skirt_front_mid_MSTR {parent=pelvis, loc=<-3.9881 9.7696 0>, rot=<0.011 0.707 0.707 0.011>}
 dyn_skirt_front_mid_01 {parent=dyn_skirt_front_mid_MSTR, loc=<0 0 0>, rot=<0 0 0 1>}
 dyn_skirt_front_mid_02 {parent=dyn_skirt_front_mid_01, loc=<19.3863 0 0>, rot=<0 -0.0072 0 -1>}
 dyn_skirt_front_mid_03 {parent=dyn_skirt_front_mid_02, loc=<19.9746 0 0>, rot=<0 0.0038 0 -1>}
 dyn_skirt_front_mid_04 {parent=dyn_skirt_front_mid_03, loc=<18.579 0 0>, rot=<0 0 0 1>}
 dyn_skirt_side_r_MSTR {parent=pelvis, loc=<-3.9588 -1.2125 15.5677>, rot=<0.0506 -0.0269 0.9982 0.0166>}
 dyn_skirt_side_r_01 {parent=dyn_skirt_side_r_MSTR, loc=<0 0 0>, rot=<0 0 0 1>}
 dyn_skirt_side_r_02 {parent=dyn_skirt_side_r_01, loc=<18.659 0 0>, rot=<-0.0342 -0.0101 0.0032 -0.9994>}
 dyn_skirt_side_r_03 {parent=dyn_skirt_side_r_02, loc=<20.9646 0 0>, rot=<0 0.0042 0 -1>}
 dyn_skirt_side_r_04 {parent=dyn_skirt_side_r_03, loc=<18.2559 0.1701 0>, rot=<0 0 0 1>}
 dyn_skirt_side_l_MSTR {parent=pelvis, loc=<-3.9593 -1.2146 -15.575>, rot=<-0.0179 -0.9986 0.001 -0.0507>}
 dyn_skirt_side_l_01 {parent=dyn_skirt_side_l_MSTR, loc=<0 0 0>, rot=<0 0 0 1>}
 dyn_skirt_side_l_02 {parent=dyn_skirt_side_l_01, loc=<18.6581 0 0>, rot=<0 -0.0102 0 -0.9999>}
 dyn_skirt_side_l_03 {parent=dyn_skirt_side_l_02, loc=<20.9615 0 0>, rot=<0 -0.0021 0 -1>}
 dyn_skirt_side_l_04 {parent=dyn_skirt_side_l_03, loc=<18.3523 0 0>, rot=<0 0 0 1>}
 dyn_skirt_back_mid_MSTR {parent=pelvis, loc=<-2.8724 -12.3899 0>, rot=<0.0002 -0.7071 0.7071 -0.0002>}
 dyn_skirt_back_mid_01 {parent=dyn_skirt_back_mid_MSTR, loc=<0 0 0>, rot=<0 0 0 1>}
 dyn_skirt_back_mid_02 {parent=dyn_skirt_back_mid_01, loc=<18.5448 0 0>, rot=<0 -0.0016 0 -1>}
 dyn_skirt_back_mid_03 {parent=dyn_skirt_back_mid_02, loc=<21.9299 0 0>, rot=<0 -0.0163 0 -0.9999>}
 dyn_skirt_back_mid_04 {parent=dyn_skirt_back_mid_03, loc=<19.4606 0 0>, rot=<0 0 0 1>}
 dyn_skirt_back_l_MSTR {parent=pelvis, loc=<-3.2875 -9.4445 -11.664>, rot=<-0.0134 -0.8819 0.4704 -0.0279>}
 dyn_skirt_back_l_01 {parent=dyn_skirt_back_l_MSTR, loc=<0 0 0>, rot=<0 0 0 1>}
 dyn_skirt_back_l_02 {parent=dyn_skirt_back_l_01, loc=<18.6737 0 0>, rot=<0 -0.0056 0.0056 -1>}
 dyn_skirt_back_l_03 {parent=dyn_skirt_back_l_02, loc=<21.4554 0 0>, rot=<0 0.0063 -0.0063 -1>}
 dyn_skirt_back_l_04 {parent=dyn_skirt_back_l_03, loc=<18.9101 -0.2648 0.2649>, rot=<0.0002 0 0 -1>}
 dyn_skirt_back_r_MSTR {parent=pelvis, loc=<-3.395 -10.3086 10.4073>, rot=<0.0354 -0.4269 0.9033 0.0236>}
 dyn_skirt_back_r_01 {parent=dyn_skirt_back_r_MSTR, loc=<0 0 0>, rot=<0 0 0 1>}
 dyn_skirt_back_r_02 {parent=dyn_skirt_back_r_01, loc=<18.6861 0 0>, rot=<0 -0.0027 -0.0019 -1>}
 dyn_skirt_back_r_03 {parent=dyn_skirt_back_r_02, loc=<21.4846 -0.0974 -0.0682>, rot=<0 0.018 0.0126 -0.9998>}
 dyn_skirt_back_r_04 {parent=dyn_skirt_back_r_03, loc=<19.0167 0.4883 0.3424>, rot=<0 0 0 1>}
 dyn_skirt_ground_collision {parent=root, loc=<0 0 0>, rot=<0 0 0 1>}
 thigh_l_up_SPR {type=Curve, loc=<0 0 0>, rot=<0 0 0 1>}
 thigh_r_up_SPR {type=Curve, loc=<0 0 0>, rot=<0 0 0 1>}
]
variables [
 Disable Floor Slope {type=bool, default=False}
 Emotes {type=bool, default=False}
 ON_ground {type=bool, default=False}
 Push Down Weight {type=double, default=0}
 TombstoneAlpha {type=double, default=0}
 TombstoneEmoteActive {type=bool, default=False}
 calf_l_up_out {type=double, default=0}
 calf_r_up_out {type=double, default=0}
 isSliding {type=bool, default=False}
 leg_adjust_alpha {type=double, default=0}
 slideAlpha {type=double, default=0}
 thigh_l_back_out {type=double, default=0}
 thigh_l_side_back_inb_out {type=double, default=0}
 thigh_l_side_out {type=double, default=0}
 thigh_l_up_across_out {type=double, default=0}
 thigh_l_up_out {type=double, default=0}
 thigh_l_up_side_inb_out {type=double, default=0}
 thigh_r_back_out {type=double, default=0}
 thigh_r_side_back_inb_out {type=double, default=0}
 thigh_r_side_out {type=double, default=0}
 thigh_r_up_across_out {type=double, default=0}
 thigh_r_up_out {type=double, default=0}
 thigh_r_up_side_inb_out {type=double, default=0}
]
