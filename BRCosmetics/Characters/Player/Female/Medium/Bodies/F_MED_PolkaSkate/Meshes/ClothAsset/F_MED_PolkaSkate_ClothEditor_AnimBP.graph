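# Cloth-editor AnimBP for the PolkaSkate character, whose Output Pose takes only Copy Pose From Mesh of the attached parent.
# An unconnected branch layers additive "Familiar" state-machine animations (idle, targeting, grip, DBNO) into a Control Rig and caches it as Base.
# Base then feeds skirt RigidBody simulation and bone offsets, per-bone AnimDynamics for laces, and hand-bone tweaks, ending unused after a front-end blend.
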AnimBlueprint /BRCosmetics/Characters/Player/Female/Medium/Bodies/F_MED_PolkaSkate/Meshes/ClothAsset/F_MED_PolkaSkate_ClothEditor_AnimBP
target_skeleton=/BRCosmetics/Characters/Player/Female/Medium/Bodies/F_MED_PolkaSkate/Meshes/F_MED_PolkaSkate_Skeleton
graphs [
 AnimGraph {nodes [
   n1 = RigidBody!(
    ComponentPose=Local To Component(LocalPose=Use cached pose 'Base'())
    Alpha=Get RBAN_Alpha() -> (RBAN_Alpha:double)
    OverrideWorldGravity:Vector="0,0,-980"
    ExternalForce
    SimSpaceSettings:Sim Space Settings Structure=(WorldAlpha=1,VelocityScaleZ=0.75,DampingAlpha=0,MaxLinearVelocity=800,MaxAngularVelocity=8,MaxLinearAcceleration=4000,MaxAngularAcceleration=40,ExternalLinearDragV=(X=0.2,Y=0.2,Z=0.1),ExternalLinearVelocity=(X=0,Y=0,Z=0),ExternalAngularVelocity=(X=0,Y=0,Z=0))
   ) -> (Pose) {
    bClampLinearTranslationLimitToRefPose=true
    bOverrideWorldGravity=true
    overridePhysicsAsset=/BRCosmetics/Characters/Player/Female/Medium/Bodies/F_MED_PolkaSkate/Meshes/F_MED_PolkaSkate_Physics1
    simulationSpace=BaseBoneSpace
    alphaBoolBlend {blendOption=Linear}
    baseBoneRef {boneName=pelvis}
    componentAppliedLinearAccClamp=<3500 3500 3500>
    overrideWorldGravity {z=-980}
    simSpaceSettings {
     dampingAlpha=0
     maxAngularAcceleration=40
     maxAngularVelocity=8
     maxLinearAcceleration=4000
     maxLinearVelocity=800
     velocityScaleZ=0.75
     worldAlpha=1
     externalLinearDragV=<0.2 0.2 0.1>
    }
   }
   n2 = "Transform (Modify) Bone - Bone: dyn_skirt_front_mid_MSTR"(ComponentPose=@n1, Alpha=1, Translation="0,0,2", Rotation="4,0,0", Scale="1,1,1") {
    rotationMode=BMM_Additive
    rotationSpace=BCS_BoneSpace
    translationMode=BMM_Additive
    translationSpace=BCS_BoneSpace
    alphaBoolBlend {blendOption=Linear}
    alphaScaleBiasClamp {bInterpResult=true, interpSpeedDecreasing=5, interpSpeedIncreasing=5}
    boneToModify {boneName=dyn_skirt_front_mid_MSTR}
    rotation {pitch=4}
    translation {z=2}
   }
   n3 = "Transform (Modify) Bone - Bone: dyn_skirt_front_r_MSTR"(ComponentPose=@n2, Alpha=1, Translation="0,0,2", Rotation="6,0,0", Scale="1,1,1") {
    rotationMode=BMM_Additive
    rotationSpace=BCS_BoneSpace
    translationMode=BMM_Additive
    translationSpace=BCS_BoneSpace
    alphaBoolBlend {blendOption=Linear}
    alphaScaleBiasClamp {bInterpResult=true, interpSpeedDecreasing=5, interpSpeedIncreasing=5}
    boneToModify {boneName=dyn_skirt_front_r_MSTR}
    rotation {pitch=6}
    translation {z=2}
   }
   n4 = "Transform (Modify) Bone - Bone: dyn_skirt_side_r_MSTR"(ComponentPose=@n3, Alpha=1, Translation="0,0,2", Rotation="0,-5,0", Scale="1,1,1") {
    rotationMode=BMM_Additive
    rotationSpace=BCS_BoneSpace
    translationMode=BMM_Additive
    translationSpace=BCS_BoneSpace
    alphaBoolBlend {blendOption=Linear}
    alphaScaleBiasClamp {bInterpResult=true, interpSpeedDecreasing=5, interpSpeedIncreasing=5}
    boneToModify {boneName=dyn_skirt_side_r_MSTR}
    rotation {yaw=-5}
    translation {z=2}
   }
   n5 = "Transform (Modify) Bone - Bone: dyn_skirt_front_l_MSTR"(ComponentPose=@n4, Alpha=1, Translation="0,0,2", Rotation="0,0,0", Scale="1,1,1") {
    rotationMode=BMM_Additive
    rotationSpace=BCS_BoneSpace
    translationMode=BMM_Additive
    translationSpace=BCS_BoneSpace
    alphaBoolBlend {blendOption=Linear}
    alphaScaleBiasClamp {bInterpResult=true, interpSpeedDecreasing=5, interpSpeedIncreasing=5}
    boneToModify {boneName=dyn_skirt_front_l_MSTR}
    translation {z=2}
   }
   n6 = "Transform (Modify) Bone - Bone: dyn_skirt_front_mid_MSTR"(ComponentPose=@n5, Alpha=1, Translation="0,0,2", Rotation="7,10,0", Scale="1,1,1") {
    rotationMode=BMM_Additive
    rotationSpace=BCS_BoneSpace
    translationMode=BMM_Additive
    translationSpace=BCS_BoneSpace
    alphaBoolBlend {blendOption=Linear}
    alphaScaleBiasClamp {bInterpResult=true, interpSpeedDecreasing=5, interpSpeedIncreasing=5}
    boneToModify {boneName=dyn_skirt_front_mid_MSTR}
    rotation {pitch=7, yaw=10}
    translation {z=2}
   }
   n7 = "Transform (Modify) Bone - Bone: dyn_skirt_front_l_MSTR"(ComponentPose=@n6, Alpha=1, Translation="0,0,2", Rotation="6,0,0", Scale="1,1,1") {
    rotationMode=BMM_Additive
    rotationSpace=BCS_BoneSpace
    translationMode=BMM_Additive
    translationSpace=BCS_BoneSpace
    alphaBoolBlend {blendOption=Linear}
    alphaScaleBiasClamp {bInterpResult=true, interpSpeedDecreasing=5, interpSpeedIncreasing=5}
    boneToModify {boneName=dyn_skirt_front_l_MSTR}
    rotation {pitch=6}
    translation {z=2}
   }
   n8 = "Transform (Modify) Bone - Bone: dyn_skirt_side_r_MSTR"(ComponentPose=@n7, Alpha=1, Translation="0,0,0", Rotation="0,5,0", Scale="1,1,1") {
    rotationMode=BMM_Additive
    rotationSpace=BCS_BoneSpace
    translationMode=BMM_Additive
    translationSpace=BCS_BoneSpace
    alphaBoolBlend {blendOption=Linear}
    alphaScaleBiasClamp {bInterpResult=true, interpSpeedDecreasing=5, interpSpeedIncreasing=5}
    boneToModify {boneName=dyn_skirt_side_r_MSTR}
    rotation {yaw=5}
   }
   n9 = Anim Dynamics - Bone: dyn_lace_l_end_l_01!(
    ComponentPose=@n8
    Alpha=1
    GravityOverride:Vector="0,0,-980"
    SimSpaceSettings:Anim Phys Sim Space Settings Structure=(SimSpaceAngularAlpha=0,MaxAngularVelocity=10000,MaxAngularAcceleration=10000,ExternalAngularVelocity=(X=0,Y=0,Z=0))
    ExternalForce="0,0,0"
   ) -> (Pose) {
    angularDampingOverride=0.99
    angularSpringConstant=15000
    bAngularSpring=true
    bOverrideAngularDamping=true
    bUseGravityOverride=true
    simulationSpace=RootRelative
    alphaBoolBlend {blendOption=Linear}
    boundBone {boneName=dyn_lace_l_end_l_01}
    componentAppliedLinearAccClamp=<3000 3000 3000>
    componentLinearAccScale=<0.2 0.2 0.2>
    componentLinearVelScale=<0.2 0.2 0.2>
    gravityOverride {z=-980}
    physicsBodyDefinitions [{
      boundBone {boneName=dyn_lace_l_end_l_01}
      boxExtents=<20 20 20>
      constraintSetup {angularLimitsMax=<0 0 10>, angularLimitsMin=<0 -40 -10>, angularTarget=<1 0 0>, linearAxesMax=<0 0 0.5>}
      localJointOffset=<5 0 0>
     }]
    simSpaceSettings {maxAngularAcceleration=100, maxAngularVelocity=10}
   }
   n10 = Anim Dynamics - Bone: dyn_lace_l_end_l_02!(
    ComponentPose=@n9
    Alpha=1
    GravityOverride:Vector="0,0,-980"
    SimSpaceSettings:Anim Phys Sim Space Settings Structure=(SimSpaceAngularAlpha=0,MaxAngularVelocity=10,MaxAngularAcceleration=100,ExternalAngularVelocity=(X=0,Y=0,Z=0))
    ExternalForce="0,0,0"
   ) -> (Pose) {
    angularDampingOverride=0.99
    angularSpringConstant=15000
    bOverrideAngularDamping=true
    bUseGravityOverride=true
    simulationSpace=RootRelative
    alphaBoolBlend {blendOption=Linear}
    boundBone {boneName=dyn_lace_l_end_l_02}
    componentAppliedLinearAccClamp=<3000 3000 3000>
    componentLinearAccScale=<0.2 0.2 0.2>
    componentLinearVelScale=<0.2 0.2 0.2>
    gravityOverride {z=-980}
    physicsBodyDefinitions [{
      boundBone {boneName=dyn_lace_l_end_l_02}
      boxExtents=<20 20 20>
      constraintSetup {angularLimitsMax=<0 16 10>, angularLimitsMin=<0 -40 -10>, angularTarget=<1 0 0>}
      localJointOffset=<5 0 0>
     }]
    simSpaceSettings {maxAngularAcceleration=100, maxAngularVelocity=10}
   }
   n11 = Anim Dynamics - Bone: dyn_lace_l_loop_l_01!(
    ComponentPose=@n10
    Alpha=1
    GravityOverride:Vector="0,0,-980"
    SimSpaceSettings:Anim Phys Sim Space Settings Structure=(SimSpaceAngularAlpha=0,MaxAngularVelocity=10,MaxAngularAcceleration=100,ExternalAngularVelocity=(X=0,Y=0,Z=0))
    ExternalForce="0,0,0"
   ) -> (Pose) {
    angularDampingOverride=0.99
    angularSpringConstant=15000
    bAngularSpring=true
    bOverrideAngularDamping=true
    bUseGravityOverride=true
    simulationSpace=RootRelative
    alphaBoolBlend {blendOption=Linear}
    boundBone {boneName=dyn_lace_l_loop_l_01}
    componentAppliedLinearAccClamp=<3000 3000 3000>
    componentLinearAccScale=<0.2 0.2 0.2>
    componentLinearVelScale=<0.2 0.2 0.2>
    gravityOverride {z=-980}
    physicsBodyDefinitions [{
      boundBone {boneName=dyn_lace_l_loop_l_01}
      boxExtents=<20 20 20>
      constraintSetup {angularLimitsMax=<0 0 2>, angularLimitsMin=<0 -40 -2>, angularTarget=<1 0 0>, linearAxesMax=<0 0 0.5>}
      localJointOffset=<5 0 0>
     }]
    simSpaceSettings {maxAngularAcceleration=100, maxAngularVelocity=10}
   }
   n12 = Anim Dynamics - Bone: dyn_lace_l_loop_l_02!(
    ComponentPose=@n11
    Alpha=1
    GravityOverride:Vector="0,0,-980"
    SimSpaceSettings:Anim Phys Sim Space Settings Structure=(SimSpaceAngularAlpha=0,MaxAngularVelocity=10,MaxAngularAcceleration=100,ExternalAngularVelocity=(X=0,Y=0,Z=0))
    ExternalForce="0,0,0"
   ) -> (Pose) {
    angularDampingOverride=0.99
    angularSpringConstant=15000
    bAngularSpring=true
    bOverrideAngularDamping=true
    bUseGravityOverride=true
    simulationSpace=RootRelative
    alphaBoolBlend {blendOption=Linear}
    boundBone {boneName=dyn_lace_l_loop_l_02}
    componentAppliedLinearAccClamp=<3000 3000 3000>
    componentLinearAccScale=<0.2 0.2 0.2>
    componentLinearVelScale=<0.2 0.2 0.2>
    gravityOverride {z=-980}
    physicsBodyDefinitions [{
      boundBone {boneName=dyn_lace_l_loop_l_02}
      boxExtents=<20 20 20>
      constraintSetup {angularLimitsMax=<0 0 2>, angularLimitsMin=<0 -40 -2>, angularTarget=<1 0 0>}
      localJointOffset=<5 0 0>
     }]
    simSpaceSettings {maxAngularAcceleration=100, maxAngularVelocity=10}
   }
   n13 = Anim Dynamics - Bone: dyn_lace_l_end_r_01!(
    ComponentPose=@n12
    Alpha=1
    GravityOverride:Vector="0,0,-980"
    SimSpaceSettings:Anim Phys Sim Space Settings Structure=(SimSpaceAngularAlpha=0,MaxAngularVelocity=10,MaxAngularAcceleration=100,ExternalAngularVelocity=(X=0,Y=0,Z=0))
    ExternalForce="0,0,0"
   ) -> (Pose) {
    angularDampingOverride=0.99
    angularSpringConstant=15000
    bAngularSpring=true
    bOverrideAngularDamping=true
    bUseGravityOverride=true
    simulationSpace=RootRelative
    alphaBoolBlend {blendOption=Linear}
    boundBone {boneName=dyn_lace_l_end_r_01}
    componentAppliedLinearAccClamp=<3000 3000 3000>
    componentLinearAccScale=<0.2 0.2 0.2>
    componentLinearVelScale=<0.2 0.2 0.2>
    gravityOverride {z=-980}
    physicsBodyDefinitions [{
      boundBone {boneName=dyn_lace_l_end_r_01}
      boxExtents=<20 20 20>
      constraintSetup {angularLimitsMax=<0 0 10>, angularLimitsMin=<0 -40 -10>, angularTarget=<1 0 0>, linearAxesMax=<0 0 0.5>}
      localJointOffset=<5 0 0>
     }]
    simSpaceSettings {maxAngularAcceleration=100, maxAngularVelocity=10}
   }
   n14 = Anim Dynamics - Bone: dyn_lace_l_end_r_02!(
    ComponentPose=@n13
    Alpha=1
    GravityOverride:Vector="0,0,-980"
    SimSpaceSettings:Anim Phys Sim Space Settings Structure=(SimSpaceAngularAlpha=0,MaxAngularVelocity=10,MaxAngularAcceleration=100,ExternalAngularVelocity=(X=0,Y=0,Z=0))
    ExternalForce="0,0,0"
   ) -> (Pose) {
    angularDampingOverride=0.99
    angularSpringConstant=15000
    bAngularSpring=true
    bOverrideAngularDamping=true
    bUseGravityOverride=true
    simulationSpace=RootRelative
    alphaBoolBlend {blendOption=Linear}
    boundBone {boneName=dyn_lace_l_end_r_02}
    componentAppliedLinearAccClamp=<3000 3000 3000>
    componentLinearAccScale=<0.2 0.2 0.2>
    componentLinearVelScale=<0.2 0.2 0.2>
    gravityOverride {z=-980}
    physicsBodyDefinitions [{
      boundBone {boneName=dyn_lace_l_end_r_02}
      boxExtents=<20 20 20>
      constraintSetup {angularLimitsMax=<0 0 10>, angularLimitsMin=<0 -40 -10>, angularTarget=<1 0 0>}
      localJointOffset=<5 0 0>
     }]
    simSpaceSettings {maxAngularAcceleration=100, maxAngularVelocity=10}
   }
   n15 = Anim Dynamics - Bone: dyn_lace_l_loop_r_01!(
    ComponentPose=@n14
    Alpha=1
    GravityOverride:Vector="0,0,-980"
    SimSpaceSettings:Anim Phys Sim Space Settings Structure=(SimSpaceAngularAlpha=0,MaxAngularVelocity=10,MaxAngularAcceleration=100,ExternalAngularVelocity=(X=0,Y=0,Z=0))
    ExternalForce="0,0,0"
   ) -> (Pose) {
    angularDampingOverride=0.99
    angularSpringConstant=15000
    bAngularSpring=true
    bOverrideAngularDamping=true
    bUseGravityOverride=true
    simulationSpace=RootRelative
    alphaBoolBlend {blendOption=Linear}
    boundBone {boneName=dyn_lace_l_loop_r_01}
    componentAppliedLinearAccClamp=<3000 3000 3000>
    componentLinearAccScale=<0.2 0.2 0.2>
    componentLinearVelScale=<0.2 0.2 0.2>
    gravityOverride {z=-980}
    physicsBodyDefinitions [{
      boundBone {boneName=dyn_lace_l_loop_r_01}
      boxExtents=<20 20 20>
      constraintSetup {angularLimitsMax=<0 0 2>, angularLimitsMin=<0 -40 -2>, angularTarget=<1 0 0>, linearAxesMax=<0 0 0.5>}
      localJointOffset=<5 0 0>
     }]
    simSpaceSettings {maxAngularAcceleration=100, maxAngularVelocity=10}
   }
   n16 = Anim Dynamics - Bone: dyn_lace_l_loop_r_02!(
    ComponentPose=@n15
    Alpha=1
    GravityOverride:Vector="0,0,-980"
    SimSpaceSettings:Anim Phys Sim Space Settings Structure=(SimSpaceAngularAlpha=0,MaxAngularVelocity=10,MaxAngularAcceleration=100,ExternalAngularVelocity=(X=0,Y=0,Z=0))
    ExternalForce="0,0,0"
   ) -> (Pose) {
    angularDampingOverride=0.99
    angularSpringConstant=15000
    bAngularSpring=true
    bOverrideAngularDamping=true
    bUseGravityOverride=true
    simulationSpace=RootRelative
    alphaBoolBlend {blendOption=Linear}
    boundBone {boneName=dyn_lace_l_loop_r_02}
    componentAppliedLinearAccClamp=<3000 3000 3000>
    componentLinearAccScale=<0.2 0.2 0.2>
    componentLinearVelScale=<0.2 0.2 0.2>
    gravityOverride {z=-980}
    physicsBodyDefinitions [{
      boundBone {boneName=dyn_lace_l_loop_r_02}
      boxExtents=<20 20 20>
      constraintSetup {angularLimitsMax=<0 0 2>, angularLimitsMin=<0 -40 -2>, angularTarget=<1 0 0>}
      localJointOffset=<5 0 0>
     }]
    simSpaceSettings {maxAngularAcceleration=100, maxAngularVelocity=10}
   }
   n17 = Anim Dynamics - Bone: dyn_lace_r_end_l_01!(
    ComponentPose=@n16
    Alpha=1
    GravityOverride:Vector="0,0,-980"
    SimSpaceSettings:Anim Phys Sim Space Settings Structure=(SimSpaceAngularAlpha=0,MaxAngularVelocity=10,MaxAngularAcceleration=100,ExternalAngularVelocity=(X=0,Y=0,Z=0))
    ExternalForce="0,0,0"
   ) -> (Pose) {
    angularDampingOverride=0.99
    angularSpringConstant=15000
    bAngularSpring=true
    bOverrideAngularDamping=true
    bUseGravityOverride=true
    simulationSpace=RootRelative
    alphaBoolBlend {blendOption=Linear}
    boundBone {boneName=dyn_lace_r_end_l_01}
    componentAppliedLinearAccClamp=<3000 3000 3000>
    componentLinearAccScale=<0.2 0.2 0.2>
    componentLinearVelScale=<0.2 0.2 0.2>
    gravityOverride {z=-980}
    physicsBodyDefinitions [{
      boundBone {boneName=dyn_lace_r_end_l_01}
      boxExtents=<20 20 20>
      constraintSetup {angularLimitsMax=<0 0 10>, angularLimitsMin=<0 -40 -10>, angularTarget=<1 0 0>, linearAxesMax=<0 0 0.5>}
      localJointOffset=<5 0 0>
     }]
    simSpaceSettings {maxAngularAcceleration=100, maxAngularVelocity=10}
   }
   n18 = Anim Dynamics - Bone: dyn_lace_r_end_l_02!(
    ComponentPose=@n17
    Alpha=1
    GravityOverride:Vector="0,0,-980"
    SimSpaceSettings:Anim Phys Sim Space Settings Structure=(SimSpaceAngularAlpha=0,MaxAngularVelocity=10,MaxAngularAcceleration=100,ExternalAngularVelocity=(X=0,Y=0,Z=0))
    ExternalForce="0,0,0"
   ) -> (Pose) {
    angularDampingOverride=0.99
    angularSpringConstant=15000
    bAngularSpring=true
    bOverrideAngularDamping=true
    bUseGravityOverride=true
    simulationSpace=RootRelative
    alphaBoolBlend {blendOption=Linear}
    boundBone {boneName=dyn_lace_r_end_l_02}
    componentAppliedLinearAccClamp=<3000 3000 3000>
    componentLinearAccScale=<0.2 0.2 0.2>
    componentLinearVelScale=<0.2 0.2 0.2>
    gravityOverride {z=-980}
    physicsBodyDefinitions [{
      boundBone {boneName=dyn_lace_r_end_l_02}
      boxExtents=<20 20 20>
      constraintSetup {angularLimitsMax=<0 0 10>, angularLimitsMin=<0 -40 -10>, angularTarget=<1 0 0>}
      localJointOffset=<5 0 0>
     }]
    simSpaceSettings {maxAngularAcceleration=100, maxAngularVelocity=10}
   }
   n19 = Anim Dynamics - Bone: dyn_lace_r_loop_l_01!(
    ComponentPose=@n18
    Alpha=1
    GravityOverride:Vector="0,0,-980"
    SimSpaceSettings:Anim Phys Sim Space Settings Structure=(SimSpaceAngularAlpha=0,MaxAngularVelocity=10,MaxAngularAcceleration=100,ExternalAngularVelocity=(X=0,Y=0,Z=0))
    ExternalForce="0,0,0"
   ) -> (Pose) {
    angularDampingOverride=0.99
    angularSpringConstant=15000
    bAngularSpring=true
    bOverrideAngularDamping=true
    bUseGravityOverride=true
    simulationSpace=RootRelative
    alphaBoolBlend {blendOption=Linear}
    boundBone {boneName=dyn_lace_r_loop_l_01}
    componentAppliedLinearAccClamp=<3000 3000 3000>
    componentLinearAccScale=<0.2 0.2 0.2>
    componentLinearVelScale=<0.2 0.2 0.2>
    gravityOverride {z=-980}
    physicsBodyDefinitions [{
      boundBone {boneName=dyn_lace_r_loop_l_01}
      boxExtents=<20 20 20>
      constraintSetup {angularLimitsMax=<0 0 3>, angularLimitsMin=<0 -40 -2>, angularTarget=<1 0 0>, linearAxesMax=<0 0 0.5>}
      localJointOffset=<5 0 0>
     }]
    simSpaceSettings {maxAngularAcceleration=100, maxAngularVelocity=10}
   }
   n20 = Anim Dynamics - Bone: dyn_lace_r_loop_l_02!(
    ComponentPose=@n19
    Alpha=1
    GravityOverride:Vector="0,0,-980"
    SimSpaceSettings:Anim Phys Sim Space Settings Structure=(SimSpaceAngularAlpha=0,MaxAngularVelocity=10,MaxAngularAcceleration=100,ExternalAngularVelocity=(X=0,Y=0,Z=0))
    ExternalForce="0,0,0"
   ) -> (Pose) {
    angularDampingOverride=0.99
    angularSpringConstant=15000
    bAngularSpring=true
    bOverrideAngularDamping=true
    bUseGravityOverride=true
    simulationSpace=RootRelative
    alphaBoolBlend {blendOption=Linear}
    boundBone {boneName=dyn_lace_r_loop_l_02}
    componentAppliedLinearAccClamp=<3000 3000 3000>
    componentLinearAccScale=<0.2 0.2 0.2>
    componentLinearVelScale=<0.2 0.2 0.2>
    gravityOverride {z=-980}
    physicsBodyDefinitions [{
      boundBone {boneName=dyn_lace_r_loop_l_02}
      boxExtents=<20 20 20>
      constraintSetup {angularLimitsMax=<0 0 3>, angularLimitsMin=<0 -40 -2>, angularTarget=<1 0 0>}
      localJointOffset=<5 0 0>
     }]
    simSpaceSettings {maxAngularAcceleration=100, maxAngularVelocity=10}
   }
   n21 = Anim Dynamics - Bone: dyn_lace_r_end_r_01!(
    ComponentPose=@n20
    Alpha=1
    GravityOverride:Vector="0,0,-980"
    SimSpaceSettings:Anim Phys Sim Space Settings Structure=(SimSpaceAngularAlpha=0,MaxAngularVelocity=10,MaxAngularAcceleration=100,ExternalAngularVelocity=(X=0,Y=0,Z=0))
    ExternalForce="0,0,0"
   ) -> (Pose) {
    angularDampingOverride=0.99
    angularSpringConstant=15000
    bAngularSpring=true
    bOverrideAngularDamping=true
    bUseGravityOverride=true
    simulationSpace=RootRelative
    alphaBoolBlend {blendOption=Linear}
    boundBone {boneName=dyn_lace_r_end_r_01}
    componentAppliedLinearAccClamp=<3000 3000 3000>
    componentLinearAccScale=<0.2 0.2 0.2>
    componentLinearVelScale=<0.2 0.2 0.2>
    gravityOverride {z=-980}
    physicsBodyDefinitions [{
      boundBone {boneName=dyn_lace_r_end_r_01}
      boxExtents=<20 20 20>
      constraintSetup {angularLimitsMax=<0 0 10>, angularLimitsMin=<0 -40 -10>, angularTarget=<1 0 0>, linearAxesMax=<0 0 0.5>}
      localJointOffset=<5 0 0>
     }]
    simSpaceSettings {maxAngularAcceleration=100, maxAngularVelocity=10}
   }
   n22 = Anim Dynamics - Bone: dyn_lace_r_end_r_02!(
    ComponentPose=@n21
    Alpha=1
    GravityOverride:Vector="0,0,-980"
    SimSpaceSettings:Anim Phys Sim Space Settings Structure=(SimSpaceAngularAlpha=0,MaxAngularVelocity=10,MaxAngularAcceleration=100,ExternalAngularVelocity=(X=0,Y=0,Z=0))
    ExternalForce="0,0,0"
   ) -> (Pose) {
    angularDampingOverride=0.99
    angularSpringConstant=15000
    bAngularSpring=true
    bOverrideAngularDamping=true
    bUseGravityOverride=true
    simulationSpace=RootRelative
    alphaBoolBlend {blendOption=Linear}
    boundBone {boneName=dyn_lace_r_end_r_02}
    componentAppliedLinearAccClamp=<3000 3000 3000>
    componentLinearAccScale=<0.2 0.2 0.2>
    componentLinearVelScale=<0.2 0.2 0.2>
    gravityOverride {z=-980}
    physicsBodyDefinitions [{
      boundBone {boneName=dyn_lace_r_end_r_02}
      boxExtents=<20 20 20>
      constraintSetup {angularLimitsMax=<0 16 10>, angularLimitsMin=<0 -40 -10>, angularTarget=<1 0 0>}
      localJointOffset=<5 0 0>
     }]
    simSpaceSettings {maxAngularAcceleration=100, maxAngularVelocity=10}
   }
   n23 = Anim Dynamics - Bone: dyn_lace_r_loop_r_01!(
    ComponentPose=@n22
    Alpha=1
    GravityOverride:Vector="0,0,-980"
    SimSpaceSettings:Anim Phys Sim Space Settings Structure=(SimSpaceAngularAlpha=0,MaxAngularVelocity=10,MaxAngularAcceleration=100,ExternalAngularVelocity=(X=0,Y=0,Z=0))
    ExternalForce="0,0,0"
   ) -> (Pose) {
    angularDampingOverride=0.99
    angularSpringConstant=15000
    bAngularSpring=true
    bOverrideAngularDamping=true
    bUseGravityOverride=true
    simulationSpace=RootRelative
    alphaBoolBlend {blendOption=Linear}
    boundBone {boneName=dyn_lace_r_loop_r_01}
    componentAppliedLinearAccClamp=<3000 3000 3000>
    componentLinearAccScale=<0.2 0.2 0.2>
    componentLinearVelScale=<0.2 0.2 0.2>
    gravityOverride {z=-980}
    physicsBodyDefinitions [{
      boundBone {boneName=dyn_lace_r_loop_r_01}
      boxExtents=<20 20 20>
      constraintSetup {angularLimitsMax=<0 0 2>, angularLimitsMin=<0 -40 -2>, angularTarget=<1 0 0>, linearAxesMax=<0 0 0.5>}
      localJointOffset=<5 0 0>
     }]
    simSpaceSettings {maxAngularAcceleration=100, maxAngularVelocity=10}
   }
   n24 = Anim Dynamics - Bone: dyn_lace_r_loop_r_02!(
    ComponentPose=@n23
    Alpha=1
    GravityOverride:Vector="0,0,-980"
    SimSpaceSettings:Anim Phys Sim Space Settings Structure=(SimSpaceAngularAlpha=0,MaxAngularVelocity=10,MaxAngularAcceleration=100,ExternalAngularVelocity=(X=0,Y=0,Z=0))
    ExternalForce="0,0,0"
   ) -> (Pose) {
    angularDampingOverride=0.99
    angularSpringConstant=15000
    bAngularSpring=true
    bOverrideAngularDamping=true
    bUseGravityOverride=true
    simulationSpace=RootRelative
    alphaBoolBlend {blendOption=Linear}
    boundBone {boneName=dyn_lace_r_loop_r_02}
    componentAppliedLinearAccClamp=<3000 3000 3000>
    componentLinearAccScale=<0.2 0.2 0.2>
    componentLinearVelScale=<0.2 0.2 0.2>
    gravityOverride {z=-980}
    physicsBodyDefinitions [{
      boundBone {boneName=dyn_lace_r_loop_r_02}
      boxExtents=<20 20 20>
      constraintSetup {angularLimitsMax=<0 0 2>, angularLimitsMin=<0 -40 -2>, angularTarget=<1 0 0>}
      localJointOffset=<5 0 0>
     }]
    simSpaceSettings {maxAngularAcceleration=100, maxAngularVelocity=10}
   }
   n25 = "Transform (Modify) Bone - Bone: deform_thing_A_wrist"!(ComponentPose=@n24, bAlphaBoolEnabled:bool=True, Translation="0,2,0", Rotation="0,0,0") -> (Pose) {
    alphaInputType=Bool
    rotationMode=BMM_Additive
    rotationSpace=BCS_BoneSpace
    translationMode=BMM_Additive
    translationSpace=BCS_BoneSpace
    alphaBoolBlend {blendOption=Linear}
    boneToModify {boneName=deform_thing_A_wrist}
    translation {y=2}
   }
   n26 = "Transform (Modify) Bone - Bone: deform_thing_A_index_metacarpal"!(ComponentPose=@n25, bAlphaBoolEnabled:bool=True, Rotation="10,0,0") -> (Pose) {
    alphaInputType=Bool
    rotationMode=BMM_Additive
    rotationSpace=BCS_BoneSpace
    translationMode=BMM_Additive
    translationSpace=BCS_BoneSpace
    alphaBoolBlend {blendOption=Linear}
    boneToModify {boneName=deform_thing_A_index_metacarpal}
    rotation {pitch=10}
   }
   n27 = "Transform (Modify) Bone - Bone: deform_thing_A_thumb_01"!(ComponentPose=@n26, bAlphaBoolEnabled:bool=True, Rotation="0,-6,0") -> (Pose) {
    alphaInputType=Bool
    rotationMode=BMM_Additive
    rotationSpace=BCS_BoneSpace
    translationMode=BMM_Additive
    translationSpace=BCS_BoneSpace
    alphaBoolBlend {blendOption=Linear}
    boneToModify {boneName=deform_thing_A_thumb_01}
    rotation {yaw=-6}
   }
   n28 = "Transform (Modify) Bone - Bone: deform_thing_A_middle_metacarpal"!(ComponentPose=@n27, bAlphaBoolEnabled:bool=True, Rotation="7,0,0") -> (Pose) {
    alphaInputType=Bool
    rotationMode=BMM_Additive
    rotationSpace=BCS_BoneSpace
    translationMode=BMM_Additive
    translationSpace=BCS_BoneSpace
    alphaBoolBlend {blendOption=Linear}
    boneToModify {boneName=deform_thing_A_middle_metacarpal}
    rotation {pitch=7}
   }
   n29 = "Transform (Modify) Bone - Bone: deform_thing_A_ring_metacarpal"!(ComponentPose=@n28, bAlphaBoolEnabled:bool=True, Rotation="8,0,0") -> (Pose) {
    alphaInputType=Bool
    rotationMode=BMM_Additive
    rotationSpace=BCS_BoneSpace
    translationMode=BMM_Additive
    translationSpace=BCS_BoneSpace
    alphaBoolBlend {blendOption=Linear}
    boneToModify {boneName=deform_thing_A_ring_metacarpal}
    rotation {pitch=8}
   }
   n30 = "Transform (Modify) Bone - Bone: deform_thing_A_ring_metacarpal"(ComponentPose=@n29, bAlphaBoolEnabled=True, Rotation="0,-2.7,0") {
    alphaInputType=Bool
    rotationMode=BMM_Additive
    rotationSpace=BCS_BoneSpace
    translationMode=BMM_Additive
    translationSpace=BCS_BoneSpace
    alphaBoolBlend {blendOption=Linear}
    boneToModify {boneName=deform_thing_A_ring_metacarpal}
    rotation {yaw=-2.7}
   }
   n31 = "Transform (Modify) Bone - Bone: deform_thing_A_pinky_metacarpal"!(ComponentPose=@n30, bAlphaBoolEnabled:bool=True, Rotation="5,0,0") -> (Pose) {
    alphaInputType=Bool
    rotationMode=BMM_Additive
    rotationSpace=BCS_BoneSpace
    translationMode=BMM_Additive
    translationSpace=BCS_BoneSpace
    alphaBoolBlend {blendOption=Linear}
    boneToModify {boneName=deform_thing_A_pinky_metacarpal}
    rotation {pitch=5}
   }
   n32 = "Transform (Modify) Bone - Bone: deform_thing_A_index_01"!(ComponentPose=@n31, bAlphaBoolEnabled:bool=True, Rotation="0,10,0") -> (Pose) {
    alphaInputType=Bool
    rotationMode=BMM_Additive
    rotationSpace=BCS_BoneSpace
    translationMode=BMM_Additive
    translationSpace=BCS_BoneSpace
    alphaBoolBlend {blendOption=Linear}
    boneToModify {boneName=deform_thing_A_index_01}
    rotation {yaw=10}
   }
   Component To Local(ComponentPose=@n32)
   Local To Component(LocalPose=Blend Poses by bool(BlendPose_0=Component To Local(), BlendPose_1=Component To Local(), bActiveValue=Get bIsInFrontEnd()))
   RigidBody(
    OverrideWorldGravity="0,0,-980"
    SimSpaceSettings=(WorldAlpha=1,VelocityScaleZ=0.75,DampingAlpha=0,MaxLinearVelocity=800,MaxAngularVelocity=8,MaxLinearAcceleration=4000,MaxAngularAcceleration=40,ExternalLinearDragV=(X=0.2,Y=0.2,Z=0.1),ExternalLinearVelocity=(X=0,Y=0,Z=0),ExternalAngularVelocity=(X=0,Y=0,Z=0))
   ) {
    bClampLinearTranslationLimitToRefPose=true
    bOverrideWorldGravity=true
    simulationSpace=BaseBoneSpace
    alphaBoolBlend {blendOption=Linear}
    alphaScaleBiasClamp {bInterpResult=true}
    baseBoneRef {boneName=pelvis}
    overrideWorldGravity {z=-980}
    simSpaceSettings {
     dampingAlpha=0
     maxAngularAcceleration=40
     maxAngularVelocity=8
     maxLinearAcceleration=4000
     maxLinearVelocity=800
     velocityScaleZ=0.75
     worldAlpha=1
     externalLinearDragV=<0.2 0.2 0.1>
    }
   }
   Output Pose(Result=Copy Pose From Mesh() {bUseAttachedParent=true}) {layerGroup=None}
   n33 = "Sequence Player (additive) 'PolkaSkate_Familiar_Base_Pose'"() {bLoopAnimation=false, sequence=/BRCosmetics/Characters/Player/Female/Medium/Bodies/F_MED_PolkaSkate/Animation/PolkaSkate_Familiar_Base_Pose}
   AnimGraphNode_StateMachine_0 = Familiar_State_1() {is=title AnimGraphNode_StateMachine}
   n34 = Control Rig!(
    Source=Apply Additive(Base=Apply Additive(Additive=@n33) {alphaBoolBlend {blendOption=Linear}}, Additive=@AnimGraphNode_StateMachine_0) {alphaBoolBlend {blendOption=Linear}}
    Alpha
    leg_adjust_alpha:double=Get leg_adjust_alpha() -> (leg_adjust_alpha:double)
    Disable Floor Slope:bool=NOT Boolean(A=Get bIsOnGround() -> (bIsOnGround:bool))
    isSliding:bool=Get bIsSliding()
    TombstoneEmoteActive:bool=Get isTombstoneEmote() -> (isTombstoneEmote:bool)
    TombstoneAlpha:double=Get TombstoneEmoteActiveAlpha() -> (TombstoneEmoteActiveAlpha:double)
    Emotes:bool=Get bIsPlayingEmote()
    ON_ground:bool=Get bIsOnGround()
   ) -> (Pose) {
    alphaBoolBlend {blendOption=Linear}
    controlRigAssetReference {blueprintRigClass=/BRCosmetics/Characters/Player/Female/Medium/Bodies/F_MED_PolkaSkate/Meshes/F_MED_PolkaSkate_CtrlRig.F_MED_PolkaSkate_CtrlRig_C}
   }
   n35 = "Transform (Modify) Bone - Bone: deform_thing_A_wrist"!(ComponentPose=Local To Component(LocalPose=@n34), bAlphaBoolEnabled=True, Translation="0,50,50") -> (Pose) {
    alphaInputType=Bool
    translationMode=BMM_Additive
    translationSpace=BCS_BoneSpace
    alphaBoolBlend {blendOption=Linear}
    boneToModify {boneName=deform_thing_A_wrist}
    translation {y=50, z=50}
   }
   n36 = "Transform (Modify) Bone - Bone: deform_thing_B_wrist"!(ComponentPose=@n35, bAlphaBoolEnabled:bool=True, Translation="0,50,50") -> (Pose) {
    alphaInputType=Bool
    translationMode=BMM_Additive
    translationSpace=BCS_BoneSpace
    alphaBoolBlend {blendOption=Linear}
    boneToModify {boneName=deform_thing_B_wrist}
    translation {y=50, z=50}
   }
   Save cached pose 'Base'(Pose=Component To Local(ComponentPose=@n36)) {cacheName=Base}
  ]}
 Familiar_State_1 {
  entry=Idle
  path=AnimGraph.AnimGraphNode_StateMachine_0.Familiar_State_1
  ids=id
  nodes [
   ToFiring() {is=title AnimStateAliasNode, also {aliases [LoopTargeting, StartFiring, StartTargeting]}}
   ToElimination() {is=title AnimStateAliasNode, also {aliases [EndTargeting, Idle, LoopTargeting, StartFiring, StartTargeting]}}
   ToReactive() {is=title AnimStateAliasNode, also {aliases [EndTargeting, Idle, LoopTargeting, StartFiring, StartTargeting]}}
   ToHardGrip() {is=title AnimStateAliasNode, also {aliases [EndTargeting, Idle, LoopTargeting, StartFiring, StartTargeting]}}
   ToHangGlide() {is=title AnimStateAliasNode, also {aliases [HardGrip,Idle]}}
   AnimStateNode_0 = Idle()
   AnimStateNode_1 = StartTargeting() {is=title AnimStateNode}
   AnimStateNode_10 = HardGrip() {is=title AnimStateNode}
   AnimStateNode_11 = HardGripIn() {is=title AnimStateNode}
   AnimStateNode_12 = HardGripOut() {is=title AnimStateNode}
   AnimStateNode_13 = HangGlide() {is=title AnimStateNode}
   AnimStateNode_2 = LoopDBNO() {is=title AnimStateNode}
   AnimStateNode_3 = EndTargeting() {is=title AnimStateNode}
   AnimStateNode_4 = StartFiring() {is=title AnimStateNode}
   AnimStateNode_5 = LoopTargeting() {is=title AnimStateNode}
   AnimStateNode_6 = "Reactive "() {is=title AnimStateNode}
   AnimStateNode_7 = StartDBNO() {is=title AnimStateNode}
   AnimStateNode_8 = EndDBNO() {is=title AnimStateNode}
   AnimStateNode_9 = Elimination() {is=title AnimStateNode}
   AnimStateTransitionNode_0 = AnimStateTransitionNode()
   AnimStateTransitionNode_1 = AnimStateTransitionNode()
   AnimStateTransitionNode_11 = AnimStateTransitionNode() {bAutomaticRuleBasedOnSequencePlayerInState=true, blendMode=Linear, crossfadeDuration=0.032}
   AnimStateTransitionNode_12 = AnimStateTransitionNode()
   AnimStateTransitionNode_13 = AnimStateTransitionNode() {bAutomaticRuleBasedOnSequencePlayerInState=true, blendMode=Linear, crossfadeDuration=0}
   AnimStateTransitionNode_14 = AnimStateTransitionNode() {bAutomaticRuleBasedOnSequencePlayerInState=true, blendMode=Linear, crossfadeDuration=0}
   AnimStateTransitionNode_16 = AnimStateTransitionNode() {bAutomaticRuleBasedOnSequencePlayerInState=true, blendMode=Linear, crossfadeDuration=0}
   AnimStateTransitionNode_17 = AnimStateTransitionNode() {blendMode=Linear, crossfadeDuration=0}
   AnimStateTransitionNode_18 = AnimStateTransitionNode() {blendMode=Linear, crossfadeDuration=0}
   AnimStateTransitionNode_19 = AnimStateTransitionNode() {blendMode=Linear, crossfadeDuration=0}
   AnimStateTransitionNode_2 = AnimStateTransitionNode() {blendMode=Linear, crossfadeDuration=0}
   AnimStateTransitionNode_21 = AnimStateTransitionNode() {bAutomaticRuleBasedOnSequencePlayerInState=true, blendMode=Linear, crossfadeDuration=0}
   AnimStateTransitionNode_22 = AnimStateTransitionNode() {blendMode=Linear, crossfadeDuration=0}
   AnimStateTransitionNode_24 = AnimStateTransitionNode() {blendMode=Linear, crossfadeDuration=0}
   AnimStateTransitionNode_25 = AnimStateTransitionNode() {blendMode=Linear, crossfadeDuration=0}
   AnimStateTransitionNode_26 = AnimStateTransitionNode() {bAutomaticRuleBasedOnSequencePlayerInState=true, blendMode=Linear, crossfadeDuration=0}
   AnimStateTransitionNode_3 = AnimStateTransitionNode() {blendMode=Linear, crossfadeDuration=0}
   AnimStateTransitionNode_4 = AnimStateTransitionNode()
   AnimStateTransitionNode_5 = AnimStateTransitionNode() {blendMode=Linear, crossfadeDuration=0}
   AnimStateTransitionNode_6 = AnimStateTransitionNode() {bAutomaticRuleBasedOnSequencePlayerInState=true, blendMode=Linear, crossfadeDuration=0}
   AnimStateTransitionNode_8 = AnimStateTransitionNode() {blendMode=Linear, crossfadeDuration=0}
   AnimStateTransitionNode_9 = AnimStateTransitionNode() {bAutomaticRuleBasedOnSequencePlayerInState=true, blendMode=Linear, crossfadeDuration=0}
  ]
  links [
   {
    dst=Idle
    rule=AnimGraph.AnimGraphNode_StateMachine_0.Familiar_State_1.AnimStateTransitionNode_26.Transition
    src=Elimination
    via=AnimStateTransitionNode_26
   }
   {
    dst=Idle
    rule=AnimGraph.AnimGraphNode_StateMachine_0.Familiar_State_1.AnimStateTransitionNode_11.Transition
    src=EndDBNO
    via=AnimStateTransitionNode_11
   }
   {
    dst=Idle
    rule=AnimGraph.AnimGraphNode_StateMachine_0.Familiar_State_1.AnimStateTransitionNode_16.Transition
    src=EndTargeting
    via=AnimStateTransitionNode_16
   }
   {
    dst=Idle
    rule=AnimGraph.AnimGraphNode_StateMachine_0.Familiar_State_1.AnimStateTransitionNode_1.Transition
    src=HangGlide
    via=AnimStateTransitionNode_1
   }
   {
    dst=HardGripOut
    rule=AnimGraph.AnimGraphNode_StateMachine_0.Familiar_State_1.AnimStateTransitionNode_22.Transition
    src=HardGrip
    via=AnimStateTransitionNode_22
   }
   {
    dst=HardGrip
    rule=AnimGraph.AnimGraphNode_StateMachine_0.Familiar_State_1.AnimStateTransitionNode_21.Transition
    src=HardGripIn
    via=AnimStateTransitionNode_21
   }
   {
    dst=Idle
    rule=AnimGraph.AnimGraphNode_StateMachine_0.Familiar_State_1.AnimStateTransitionNode_14.Transition
    src=HardGripOut
    via=AnimStateTransitionNode_14
   }
   {
    dst=StartDBNO
    rule=AnimGraph.AnimGraphNode_StateMachine_0.Familiar_State_1.AnimStateTransitionNode_17.Transition
    src=Idle
    via=AnimStateTransitionNode_17
   }
   {
    dst=StartTargeting
    rule=AnimGraph.AnimGraphNode_StateMachine_0.Familiar_State_1.AnimStateTransitionNode_0.Transition
    src=Idle
    via=AnimStateTransitionNode_0
   }
   {
    dst=EndDBNO
    rule=AnimGraph.AnimGraphNode_StateMachine_0.Familiar_State_1.AnimStateTransitionNode_12.Transition
    src=LoopDBNO
    via=AnimStateTransitionNode_12
   }
   {
    dst=EndTargeting
    rule=AnimGraph.AnimGraphNode_StateMachine_0.Familiar_State_1.AnimStateTransitionNode_5.AnimationTransitionGraph_0
    src=LoopTargeting
    via=AnimStateTransitionNode_5
   }
   {
    dst=Idle
    rule=AnimGraph.AnimGraphNode_StateMachine_0.Familiar_State_1.AnimStateTransitionNode_13.Transition
    src="Reactive "
    via=AnimStateTransitionNode_13
   }
   {
    dst=LoopDBNO
    rule=AnimGraph.AnimGraphNode_StateMachine_0.Familiar_State_1.AnimStateTransitionNode_9.Transition
    src=StartDBNO
    via=AnimStateTransitionNode_9
   }
   {
    dst=EndTargeting
    rule=AnimGraph.AnimGraphNode_StateMachine_0.Familiar_State_1.AnimStateTransitionNode_8.Transition
    src=StartFiring
    via=AnimStateTransitionNode_8
   }
   {
    dst=LoopTargeting
    rule=AnimGraph.AnimGraphNode_StateMachine_0.Familiar_State_1.AnimStateTransitionNode_3.Transition
    src=StartFiring
    via=AnimStateTransitionNode_3
   }
   {
    dst=Idle
    rule=AnimGraph.AnimGraphNode_StateMachine_0.Familiar_State_1.AnimStateTransitionNode_4.Transition
    src=StartTargeting
    via=AnimStateTransitionNode_4
   }
   {
    dst=LoopTargeting
    rule=AnimGraph.AnimGraphNode_StateMachine_0.Familiar_State_1.AnimStateTransitionNode_6.Transition
    src=StartTargeting
    via=AnimStateTransitionNode_6
   }
   {
    dst=Elimination
    rule=AnimGraph.AnimGraphNode_StateMachine_0.Familiar_State_1.AnimStateTransitionNode_18.Transition
    src=ToElimination
    via=AnimStateTransitionNode_18
   }
   {
    dst=StartFiring
    rule=AnimGraph.AnimGraphNode_StateMachine_0.Familiar_State_1.AnimStateTransitionNode_2.Transition
    src=ToFiring
    via=AnimStateTransitionNode_2
   }
   {
    dst=HangGlide
    rule=AnimGraph.AnimGraphNode_StateMachine_0.Familiar_State_1.AnimStateTransitionNode_25.Transition
    src=ToHangGlide
    via=AnimStateTransitionNode_25
   }
   {
    dst=HardGripIn
    rule=AnimGraph.AnimGraphNode_StateMachine_0.Familiar_State_1.AnimStateTransitionNode_24.Transition
    src=ToHardGrip
    via=AnimStateTransitionNode_24
   }
   {
    dst="Reactive "
    rule=AnimGraph.AnimGraphNode_StateMachine_0.Familiar_State_1.AnimStateTransitionNode_19.Transition
    src=ToReactive
    via=AnimStateTransitionNode_19
   }
  ]
 }
 Idle {
  path=AnimGraph.AnimGraphNode_StateMachine_0.Familiar_State_1.AnimStateNode_0.Idle
  nodes [
   n1 = Random Sequence Player!() -> (Pose:Pose Link Structure) {
    is=title AnimGraphNode_RandomPlayer
    entries [
     {
      chanceToPlay=0.75
      maxLoopCount=0
      maxPlayRate=1
      minLoopCount=0
      minPlayRate=1
      sequence=/BRCosmetics/Characters/Player/Female/Medium/Bodies/F_MED_PolkaSkate/Animation/PolkaSkate_Familiar_IdleLookabout
      blendIn {blendOption=Linear, blendTime=0.2, customCurve=null}
     }
     {
      chanceToPlay=0.25
      maxLoopCount=0
      maxPlayRate=1
      minLoopCount=0
      minPlayRate=1
      sequence=/BRCosmetics/Characters/Player/Female/Medium/Bodies/F_MED_PolkaSkate/Animation/PolkaSkate_Familiar_IdleTapTapTap
      blendIn {blendOption=Linear, blendTime=0.2, customCurve=null}
     }
    ]
   }
   Output Animation Pose(Result=@n1)
  ]
 }
 StartTargeting {
  path=AnimGraph.AnimGraphNode_StateMachine_0.Familiar_State_1.AnimStateNode_1.StartTargeting
  nodes [
   n1 = "Sequence Player (additive) 'PolkaSkate_Familiar_ADS_Into'"() {bLoopAnimation=false, sequence=/BRCosmetics/Characters/Player/Female/Medium/Bodies/F_MED_PolkaSkate/Animation/PolkaSkate_Familiar_ADS_Into}
   Output Animation Pose(Result=@n1)
  ]
 }
 HardGrip {
  path=AnimGraph.AnimGraphNode_StateMachine_0.Familiar_State_1.AnimStateNode_10.HardGrip
  nodes [
   n1 = "Sequence Player (additive) 'PolkaSkate_Familiar_Grip_Jitter'"() {sequence=/BRCosmetics/Characters/Player/Female/Medium/Bodies/F_MED_PolkaSkate/Animation/PolkaSkate_Familiar_Grip_Jitter}
   n2 = Get bIsSkyDiving()
   n3 = "Sequence Player (additive) 'PolkaSkate_Familiar_Grip_Pose'"() {bLoopAnimation=false, sequence=/BRCosmetics/Characters/Player/Female/Medium/Bodies/F_MED_PolkaSkate/Animation/PolkaSkate_Familiar_Grip_Pose}
   Output Animation Pose(Result=Apply Additive(Base=@n3, Additive=Blend Poses by bool(BlendPose_0=@n1, bActiveValue=@n2) {blendType=Linear}) {alphaBoolBlend {blendOption=Linear}})
  ]
 }
 HardGripIn {
  path=AnimGraph.AnimGraphNode_StateMachine_0.Familiar_State_1.AnimStateNode_11.HardGripIn
  nodes [
   n1 = "Sequence Player (additive) 'PolkaSkate_Familiar_GripInto_Anim'"() {
    bLoopAnimation=false
    sequence=/BRCosmetics/Characters/Player/Female/Medium/Bodies/F_MED_PolkaSkate/Animation/PolkaSkate_Familiar_GripInto_Anim
   }
   Output Animation Pose(Result=@n1)
  ]
 }
 HardGripOut {
  path=AnimGraph.AnimGraphNode_StateMachine_0.Familiar_State_1.AnimStateNode_12.HardGripOut
  nodes [
   n1 = "Sequence Player (additive) 'PolkaSkate_Familiar_GripOut_Anim'"() {
    bLoopAnimation=false
    playRate=2
    sequence=/BRCosmetics/Characters/Player/Female/Medium/Bodies/F_MED_PolkaSkate/Animation/PolkaSkate_Familiar_GripOut_Anim
   }
   Output Animation Pose(Result=@n1)
  ]
 }
 HangGlide {
  path=AnimGraph.AnimGraphNode_StateMachine_0.Familiar_State_1.AnimStateNode_13.HangGlide
  nodes [
   n1 = "Sequence Player (additive) 'PolkaSkate_Familiar_HanggliderState_Anim'"() {sequence=/BRCosmetics/Characters/Player/Female/Medium/Bodies/F_MED_PolkaSkate/Animation/PolkaSkate_Familiar_HanggliderState_Anim}
   n2 = "Sequence Player (additive) 'PolkaSkate_Familiar_HanggliderState_Pose'"() {sequence=/BRCosmetics/Characters/Player/Female/Medium/Bodies/F_MED_PolkaSkate/Animation/PolkaSkate_Familiar_HanggliderState_Pose}
   n3 = "Sequence Player (additive) 'PolkaSkate_Familiar_Grip_Jitter'"() {
    bLoopAnimation=false
    sequence=&PolkaSkate_Familiar_Grip_Jitter
   }
   Output Animation Pose(Result=Apply Additive(Base=Apply Additive(Base=@n2, Additive=@n1) {alphaBoolBlend {blendOption=Linear}}, Additive=@n3) {alphaBoolBlend {blendOption=Linear}})
  ]
 }
 LoopDBNO {
  path=AnimGraph.AnimGraphNode_StateMachine_0.Familiar_State_1.AnimStateNode_2.LoopDBNO
  nodes [Output Animation Pose(Result="Sequence Player (additive) 'PolkaSkate_Familiar_Grip_Pose'"() {sequence=&PolkaSkate_Familiar_Grip_Pose})]
 }
 EndTargeting {
  path=AnimGraph.AnimGraphNode_StateMachine_0.Familiar_State_1.AnimStateNode_3.EndTargeting
  nodes [
   n1 = "Sequence Player (additive) 'PolkaSkate_Familiar_ADS_Out'"() {
    bLoopAnimation=false
    playRate=2
    sequence=/BRCosmetics/Characters/Player/Female/Medium/Bodies/F_MED_PolkaSkate/Animation/PolkaSkate_Familiar_ADS_Out
   }
   Output Animation Pose(Result=@n1)
  ]
 }
 StartFiring {
  path=AnimGraph.AnimGraphNode_StateMachine_0.Familiar_State_1.AnimStateNode_4.StartFiring
  nodes [Output Animation Pose(Result="Sequence Player (additive) 'PolkaSkate_Familiar_ADS_Fire'"() {sequence=/BRCosmetics/Characters/Player/Female/Medium/Bodies/F_MED_PolkaSkate/Animation/PolkaSkate_Familiar_ADS_Fire})]
 }
 LoopTargeting {
  path=AnimGraph.AnimGraphNode_StateMachine_0.Familiar_State_1.AnimStateNode_5.LoopTargeting
  nodes [
   n1 = "Sequence Player (additive) 'PolkaSkate_Familiar_ADS_Pose'"() {bLoopAnimation=false, sequence=/BRCosmetics/Characters/Player/Female/Medium/Bodies/F_MED_PolkaSkate/Animation/PolkaSkate_Familiar_ADS_Pose}
   Output Animation Pose(Result=@n1)
  ]
 }
 "Reactive " {
  path="AnimGraph.AnimGraphNode_StateMachine_0.Familiar_State_1.AnimStateNode_6.Reactive "
  nodes [
   n1 = "Sequence Player (additive) 'PolkaSkate_Familiar_Reactive_SnapSnap'"() {
    bLoopAnimation=false
    sequence=/BRCosmetics/Characters/Player/Female/Medium/Bodies/F_MED_PolkaSkate/Animation/PolkaSkate_Familiar_Reactive_SnapSnap
   }
   Output Animation Pose(Result=@n1)
  ]
 }
 StartDBNO {
  path=AnimGraph.AnimGraphNode_StateMachine_0.Familiar_State_1.AnimStateNode_7.StartDBNO
  nodes [
   n1 = "Sequence Player (additive) 'PolkaSkate_Familiar_GripInto_Anim'"() {
    bLoopAnimation=false
    sequence=&PolkaSkate_Familiar_GripInto_Anim
   }
   Output Animation Pose(Result=@n1)
  ]
 }
 EndDBNO {
  path=AnimGraph.AnimGraphNode_StateMachine_0.Familiar_State_1.AnimStateNode_8.EndDBNO
  nodes [
   n1 = "Sequence Player (additive) 'PolkaSkate_Familiar_GripOut_Anim'"() {
    bLoopAnimation=false
    sequence=&PolkaSkate_Familiar_GripOut_Anim
   }
   Output Animation Pose(Result=@n1)
  ]
 }
 Elimination {
  path=AnimGraph.AnimGraphNode_StateMachine_0.Familiar_State_1.AnimStateNode_9.Elimination
  nodes [
   n1 = "Sequence Player (additive) 'PolkaSkate_Familiar_Reactive_ThumbsUp'"() {
    bLoopAnimation=false
    sequence=/BRCosmetics/Characters/Player/Female/Medium/Bodies/F_MED_PolkaSkate/Animation/PolkaSkate_Familiar_Reactive_ThumbsUp
   }
   Output Animation Pose(Result=@n1)
  ]
 }
 Transition {
  path=AnimGraph.AnimGraphNode_StateMachine_0.Familiar_State_1.AnimStateTransitionNode_0.Transition
  nodes [Result(bCanEnterTransition=Get bIsTargeting())]
 }
 Transition {
  path=AnimGraph.AnimGraphNode_StateMachine_0.Familiar_State_1.AnimStateTransitionNode_1.Transition
  nodes [Result(bCanEnterTransition=NOT Boolean(A=Get bIsParachuteOpened()))]
 }
 Transition {path=AnimGraph.AnimGraphNode_StateMachine_0.Familiar_State_1.AnimStateTransitionNode_11.Transition, nodes [Result()]}
 Transition {
  path=AnimGraph.AnimGraphNode_StateMachine_0.Familiar_State_1.AnimStateTransitionNode_12.Transition
  nodes [Result(bCanEnterTransition=NOT Boolean(A=Get bIsDBNO()))]
 }
 Transition {path=AnimGraph.AnimGraphNode_StateMachine_0.Familiar_State_1.AnimStateTransitionNode_13.Transition, nodes [Result()]}
 Transition {path=AnimGraph.AnimGraphNode_StateMachine_0.Familiar_State_1.AnimStateTransitionNode_14.Transition, nodes [Result()]}
 Transition {path=AnimGraph.AnimGraphNode_StateMachine_0.Familiar_State_1.AnimStateTransitionNode_16.Transition, nodes [Result()]}
 Transition {
  path=AnimGraph.AnimGraphNode_StateMachine_0.Familiar_State_1.AnimStateTransitionNode_17.Transition
  nodes [Result(bCanEnterTransition=Get bIsDBNO())]
 }
 Transition {
  path=AnimGraph.AnimGraphNode_StateMachine_0.Familiar_State_1.AnimStateTransitionNode_18.Transition
  nodes [Result(bCanEnterTransition=Get IsElimActive() -> (IsElimActive:bool))]
 }
 Transition {
  path=AnimGraph.AnimGraphNode_StateMachine_0.Familiar_State_1.AnimStateTransitionNode_19.Transition
  nodes [Result(bCanEnterTransition=Get isReactive() -> (isReactive:bool))]
 }
 Transition {
  path=AnimGraph.AnimGraphNode_StateMachine_0.Familiar_State_1.AnimStateTransitionNode_2.Transition
  nodes [Result(bCanEnterTransition=Get IsFiring() -> (IsFiring:bool))]
 }
 Transition {path=AnimGraph.AnimGraphNode_StateMachine_0.Familiar_State_1.AnimStateTransitionNode_21.Transition, nodes [Result()]}
 Transition {
  path=AnimGraph.AnimGraphNode_StateMachine_0.Familiar_State_1.AnimStateTransitionNode_22.Transition
  nodes [Result(bCanEnterTransition=NOR Boolean!(A:bool=Get bIsSkyDiving(), B:bool=Get bIsTacticalSprinting()) -> (ReturnValue:bool))]
 }
 Transition {
  path=AnimGraph.AnimGraphNode_StateMachine_0.Familiar_State_1.AnimStateTransitionNode_24.Transition
  nodes [Result(bCanEnterTransition=OR Boolean(A=Get bIsSkyDiving(), B=Get bIsTacticalSprinting()))]
 }
 Transition {
  path=AnimGraph.AnimGraphNode_StateMachine_0.Familiar_State_1.AnimStateTransitionNode_25.Transition
  nodes [Result(bCanEnterTransition=Get bIsParachuteOpened())]
 }
 Transition {path=AnimGraph.AnimGraphNode_StateMachine_0.Familiar_State_1.AnimStateTransitionNode_26.Transition, nodes [Result()]}
 Transition {
  path=AnimGraph.AnimGraphNode_StateMachine_0.Familiar_State_1.AnimStateTransitionNode_3.Transition
  nodes [Result(bCanEnterTransition=NOT Boolean(A=Get IsFiring()))]
 }
 Transition {
  path=AnimGraph.AnimGraphNode_StateMachine_0.Familiar_State_1.AnimStateTransitionNode_4.Transition
  nodes [Result(bCanEnterTransition=NOT Boolean(A=Get bIsTargeting()))]
 }
 AnimationTransitionGraph_0 {
  path=AnimGraph.AnimGraphNode_StateMachine_0.Familiar_State_1.AnimStateTransitionNode_5.AnimationTransitionGraph_0
  nodes [Result(bCanEnterTransition=NOT Boolean(A=Get bIsTargeting()))]
 }
 Transition {path=AnimGraph.AnimGraphNode_StateMachine_0.Familiar_State_1.AnimStateTransitionNode_6.Transition, nodes [Result()]}
 Transition {
  path=AnimGraph.AnimGraphNode_StateMachine_0.Familiar_State_1.AnimStateTransitionNode_8.Transition
  nodes [Result(bCanEnterTransition=NOT Boolean(A=Get bIsTargeting()))]
 }
 Transition {path=AnimGraph.AnimGraphNode_StateMachine_0.Familiar_State_1.AnimStateTransitionNode_9.Transition, nodes [Result()]}
 EventGraph {nodes [
   n1 = Get Dynamic Material Instance(execute=@n26.then, SkeletalMeshComponent=Get BodySkelPart() -> (BodySkelPart:Skeletal Mesh Component Object Reference))
   > n2 = Set DynamicMaterialInstance(DynamicMaterialInstance:Material Instance Dynamic Object Reference=@n1) -> (Output_Get:Material Instance Dynamic Object Reference)
   SetScalarParameterValue(self=[@n1, @n2], ParameterName=Hide Element 01, Value=0)
   > SetScalarParameterValue(self=[@n1, @n2], ParameterName=Hide Element 02, Value=1)
   Event BlueprintUpdateAnimation()
   > n3 = Sequence!(execute) -> (then_0, then_1, then_2:Exec, then_3:Exec, then_4:Exec, then_5:Exec, then_6:Exec, then_7:Exec)
   n4 = GetOwnerFortPlayerAnimInstance!(self:Custom Character Part Anim Instance Object Reference=Get CharacterPartAnimInstance() -> (CharacterPartAnimInstance:Custom Character Part Anim Instance Object Reference)) -> (ReturnValue:Fort Player Anim Instance Object Reference)
   > Is Valid(InputObject=@n4)
   > n5 = Branch(Condition=Get bIsParachuteOpened())
   > n6 = Switch on EGliderType(Selection=Break Fort Anim Input Player Glider Anim Asset!(FortAnimInput_PlayerGliderAnimAsset:"Fort Anim Input Player Glider Anim Asset Structure (by ref)"=Get CurrentGliderAnimSet!(self:Fort Player Anim Instance Object Reference=@n4) -> (CurrentGliderAnimSet:Fort Anim Input Player Glider Anim Asset Structure)) -> (
     Default_Base_BS:Blend Space Object Reference
     Default_BodyAdditive_MaleMedium_BS:Blend Space Object Reference
     Default_BodyAdditive_MaleLarge_BS:Blend Space Object Reference
     Default_BodyAdditive_FemaleSmall_BS:Blend Space Object Reference
     Default_BodyAdditive_FemaleMedium_BS:Blend Space Object Reference
     Default_BodyAdditive_FemaleLarge_BS:Blend Space Object Reference
     Default_TurnAdditive_MaleMedium_BS:Blend Space Object Reference
     Default_TurnAdditive_MaleLarge_BS:Blend Space Object Reference
     Default_TurnAdditive_FemaleSmall_BS:Blend Space Object Reference
     Default_TurnAdditive_FemaleMedium_BS:Blend Space Object Reference
     Default_TurnAdditive_FemaleLarge_BS:Blend Space Object Reference
     Into_Base_BS:Blend Space Object Reference
     Into_Base_MaleMedium_BS:Blend Space Object Reference
     Into_BodyAdditive_MaleMedium_BS:Blend Space Object Reference
     Into_BodyAdditive_MaleLarge_BS:Blend Space Object Reference
     Into_BodyAdditive_FemaleSmall_BS:Blend Space Object Reference
     Into_BodyAdditive_FemaleMedium_BS:Blend Space Object Reference
     Into_BodyAdditive_FemaleLarge_BS:Blend Space Object Reference
     Into_TurnAdditive_MaleMedium_BS:Blend Space Object Reference
     Into_TurnAdditive_MaleLarge_BS:Blend Space Object Reference
     Into_TurnAdditive_FemaleSmall_BS:Blend Space Object Reference
     Into_TurnAdditive_FemaleMedium_BS:Blend Space Object Reference
     Into_TurnAdditive_FemaleLarge_BS:Blend Space Object Reference
     Lean_MaleMedium_BS:Blend Space Object Reference
     Lean_MaleLarge_BS:Blend Space Object Reference
     Lean_FemaleSmall_BS:Blend Space Object Reference
     Lean_FemaleMedium_BS:Blend Space Object Reference
     Lean_FemaleLarge_BS:Blend Space Object Reference
     LeanAdditive_Center_MaleMedium_Pose:Anim Sequence Object Reference
     LeanAdditive_Center_MaleLarge_Pose:Anim Sequence Object Reference
     LeanAdditive_Center_FemaleSmall_Pose:Anim Sequence Object Reference
     LeanAdditive_Center_FemaleMedium_Pose:Anim Sequence Object Reference
     LeanAdditive_Center_FemaleLarge_Pose:Anim Sequence Object Reference
     LeanAdditive_Into_BS:Blend Space Object Reference
     LeanAdditive_ForwardInto_Anim:Anim Sequence Object Reference
     LeanAdditive_ForwardInto_FromDeploy_Anim:Anim Sequence Object Reference
     LeanAdditive_BackInto_Anim:Anim Sequence Object Reference
     LeanAdditive_LeftInto_Anim:Anim Sequence Object Reference
     LeanAdditive_RightInto_Anim:Anim Sequence Object Reference
     LeanAdditive_ForwardOut_Anim:Anim Sequence Object Reference
     LeanAdditive_BackOut_Anim:Anim Sequence Object Reference
     LeanAdditive_LeftOut_Anim:Anim Sequence Object Reference
     LeanAdditive_RightOut_Anim:Anim Sequence Object Reference
     ToGlide_BS:Blend Space Object Reference
     ToGlide_Lean_BS:Blend Space Object Reference
     ToDive_BS:Blend Space Object Reference
     ToDive_Lean_BS:Blend Space Object Reference
     Dive_WeaponR_Additive_BS:Blend Space Object Reference
     Glide_WeaponR_Additive_Anim:Anim Sequence Object Reference
     GenericAdditive_Male_BS:Blend Space Object Reference
     GenericAdditive_Female_BS:Blend Space Object Reference
     RootModPitchMin:float=~
     RootModPitchMax:float=~
     RootModYOffsetMin:float=~
     RootModYOffsetMax:float=~
     PelvisModVertical:Pelvis Mod Vertical Input Spring Structure
     PelvisModLateral:Pelvis Mod Lateral Input Spring Structure
     PelvisModBankAngle:Pelvis Mod Bank Angle Input Spring Structure
     ThighModLegAngle:Thigh Mod Leg Angle Input Structure
     ThighModLegPitch:Thigh Mod Leg Pitch Spring Structure
     ThighModLegBank:Thigh Mod Leg Bank Spring Structure
     ThighModLeftLegDrag:Thigh Mod Left Leg Drag Structure
     PlayerGliderType:EGliderType Enum
     bEnableSpringMods:bool=~
     bAllowPlayerDeployRootMod:bool=~
     bUseSurfStyle:bool=~
     bUseLeanPlayRateAdjust:bool=~
     bUseForwardAs_Y_BS:bool=~
     bUseDeployCurveForIntoPlayRateAdjust:bool=~
     LeanForwardValue:float=~
     IdleValue:float=~
     LeanBackValue:float=~
     IdleInterpToSpeed:float=~
     LeanForwardInterpToSpeed:float=~
     LeanBackInterpToSpeed:float=~
    ) {is=title K2Node_BreakStruct}.PlayerGliderType)
   Set hideMeshParts(execute=[@n6.Cape, @n6.HangGlider], hideMeshParts:bool=true) -> (Output_Get:bool)
   Set hideMeshParts(execute=[@n6.NoGlider, @n6.Stairs, @n6.Surfing, @n6.SurfingSimple, @n6.Umbrella], hideMeshParts=false)
   n7 = Branch(execute=@n5.else, Condition=Get isTombstoneEmote())
   > Set hideMeshParts(hideMeshParts=true)
   Set hideMeshParts(execute=@n7.else, hideMeshParts=false)
   n8 = GetAnimInstance(self=GetOwnerSkeletalMeshComponent())
   Is Valid(exec=@n3.then_5, InputObject=@n8)
   n9 = GetCurveValue(self=@n8, CurveName=EmoteActive)
   > Set TombstoneEmoteActiveAlpha(TombstoneEmoteActiveAlpha:double=@n9) -> (Output_Get:double)
   > n10 = Branch(Condition="float >= float"(A=@n9, B=1))
   > Set isTombstoneEmote(isTombstoneEmote:bool=true) -> (Output_Get:bool)
   Set isTombstoneEmote(execute=@n10.else, isTombstoneEmote=false)
   n11 = Branch(execute=@n3.then_6, Condition=Get isTombstoneEmote())
   > Set RBAN_Alpha(RBAN_Alpha:double=0.2) -> (Output_Get:double)
   Set RBAN_Alpha(execute=@n11.else, RBAN_Alpha=1)
   n12 = Branch(execute=@n3.then_1, Condition=Get hideMeshParts() -> (hideMeshParts:bool))
   > n13 = Branch(Condition=Get isTombstoneEmote())
   n14 = Get DynamicMaterialInstance() -> (DynamicMaterialInstance:Material Instance Dynamic Object Reference)
   > SetScalarParameterValue(self=@n14, ParameterName=Hide Element 01, Value=1)
   > SetScalarParameterValue(self=@n14, ParameterName=Hide Element 02, Value=1)
   SetScalarParameterValue(execute=@n13.else, self=@n14, ParameterName=Hide Element 01, Value=1)
   n15 = CharacterItemDefinitionHasAnyMetaTag(Tags=Make Array("[0]"=(TagName="Cosmetics.Variant.Meta.PolkaSkate.ThingOn")))
   > n16 = Branch(Condition=@n15)
   > SetScalarParameterValue(self=@n14, ParameterName=Hide Element 02, Value=0)
   SetScalarParameterValue(execute=@n16.else, self=@n14, ParameterName=Hide Element 02, Value=1)
   SetScalarParameterValue(execute=@n12.else, self=@n14, ParameterName=Hide Element 02, Value=1)
   > n17 = Branch(Condition=@n15)
   > SetScalarParameterValue(self=@n14, ParameterName=Hide Element 01, Value=0)
   SetScalarParameterValue(execute=@n17.else, self=@n14, ParameterName=Hide Element 01, Value=1)
   Set IsFiring(execute=@n3.then_2, IsFiring:bool=Get bRecentlyFired() -> (bRecentlyFired:bool)) -> (Output_Get:bool)
   n18 = "Gravity Override State Function (Parameter Asset) (Nativized)"(
    execute=@n3.then_3
    CharacterPartAnimInstance=Self-Reference()
    GameTimeInSeconds=GetGameTimeInSeconds()
    Parameters=/BRCosmetics/Characters/Player/Female/Medium/Bodies/F_MED_PolkaSkate/Meshes/F_MED_PolkaSkate_Parameters
   )
   > n19 = Branch(Condition=Get bIsSliding())
   > n20 = Set Gravity Override(Gravity Override="0,0,-100")
   Set RBAN Settings Out(execute=[@n22.then, @n20.then], RBAN Settings Out=@n18.OutRigidBodyAnimNodeParameters)
   n21 = Branch(execute=@n19.else, Condition=Get bIsCrouching())
   > Set Gravity Override(Gravity Override="0,0,-600")
   n22 = Set Gravity Override(execute=@n21.else, Gravity Override=@n18)
   n23 = Branch(execute=@n3.then_4, Condition=NOT Boolean(A=Get bIsOnGround()))
   Set thigh_l_up_alpha(execute=[@n24.then, @n23.then], thigh_l_up_alpha:double=1) -> (Output_Get:double)
   > Set thigh_r_up_alpha(thigh_r_up_alpha:double=1) -> (Output_Get:double)
   n24 = Branch(execute=@n23.else, Condition=Get isTombstoneEmote())
   n25 = Branch(execute=@n24.else, Condition=Get bIsCrouching())
   > Set thigh_l_up_alpha(thigh_l_up_alpha=0)
   > Set thigh_r_up_alpha(thigh_r_up_alpha=0)
   Set thigh_l_up_alpha(execute=@n25.else, thigh_l_up_alpha=GetCurveValue(CurveName=thigh_l_up_SPR))
   > Set thigh_r_up_alpha(thigh_r_up_alpha=GetCurveValue(CurveName=thigh_r_up_SPR))
   Event OnCharacterCustomizationLoaded()
   > Set BodySkelPart(BodySkelPart:Skeletal Mesh Component Object Reference=GetPartSkeletalMeshComponent()) -> (Output_Get:Skeletal Mesh Component Object Reference)
   > n26 = Set CharacterPartAnimInstance(CharacterPartAnimInstance:Custom Character Part Anim Instance Object Reference=Self-Reference()) -> (Output_Get:Custom Character Part Anim Instance Object Reference)
   Event OnBeginSkydiving()
   > Set deferred_helper_alpha(deferred_helper_alpha:double=0) -> (Output_Get:double)
   > Set leg_adjust_alpha(leg_adjust_alpha:double=0) -> (Output_Get:double)
   Event OnEndSkydiving()
   > Set deferred_helper_alpha(deferred_helper_alpha=1)
   > Set leg_adjust_alpha(leg_adjust_alpha=1)
   > n27 = Branch(Condition=NOT Boolean(A=Get bIsOnGround()))
   > Set leg_adjust_alpha(leg_adjust_alpha=0)
   Set leg_adjust_alpha(execute=@n27.else, leg_adjust_alpha=1)
   Event OnBeginPlayingEmote() {is=title K2Node_Event}
   > Set leg_adjust_alpha(leg_adjust_alpha=1)
   Event OnEndPlayingEmote() {is=title K2Node_Event}
   > Set leg_adjust_alpha(leg_adjust_alpha=1)
   Event OnBeginJumping()
   > Set leg_adjust_alpha(leg_adjust_alpha=0)
   Event OnEndJumping()
   > Set leg_adjust_alpha(leg_adjust_alpha=1)
   Event OnBeginTargeting() {is=title K2Node_Event}
   > Set isTargeting(isTargeting:bool=true) -> (Output_Get:bool)
   > Branch(Condition=Get shouldAdjustSkirtFront() -> (shouldAdjustSkirtFront:bool))
   > SetSkinWeightProfile!(
    execute
    self:Skinned Mesh Component Object Reference=Get BodySkelPart()
    InProfileName:Name=Alternative
    InLayer:ESkinWeightProfileLayer Enum=Secondary
   ) -> (then, ReturnValue:bool)
   Event OnEndTargeting() {is=title K2Node_Event}
   > Set isTargeting(isTargeting=false)
   > SetSkinWeightProfile(self=Get BodySkelPart(), InProfileName=None, InLayer=Secondary)
   n28 = Event BlueprintProcessAnimInputEvent()
   n29 = IsNamed(self=@n28.AnimInputEvent, InEventName=aie.ChestOpened)
   n30 = IsNamed(self=@n28.AnimInputEvent, InEventName=aie.EnemyPlayerEliminated)
   n31 = Get CurrentWeapon() -> (CurrentWeapon:Fort Weapon Object Reference)
   n32 = AND Boolean(
    A=IsNamed(self=@n28.AnimInputEvent, InEventName=aie.PlayerVitality.ShieldsIncreased)
    B=Contains(SearchIn=GetDisplayName(Object=@n31), Substring=SlurpJuice)
   )
   > Branch(Condition=AND Boolean(A=OR Boolean!(A=@n30, B=@n29, C:bool=@n32) -> (ReturnValue), B=Get ReactivityEnabled() -> (ReactivityEnabled:bool)))
   > n33 = Branch(Condition=NOT Boolean(A=@n30))
   > Set isReactive(isReactive:bool=true) -> (Output_Get:bool)
   Set IsElimActive(execute=@n33.else, IsElimActive:bool=true) -> (Output_Get:bool)
   Event BlueprintPostEvaluateAnimation() {is=title K2Node_Event}
   > Set isReactive(isReactive=false)
   > Set IsElimActive(IsElimActive=false)
   Event OnBeginCrouching()
   > Set leg_adjust_alpha(leg_adjust_alpha=0)
   > Set shouldAdjustSkirtFront(shouldAdjustSkirtFront:bool=true) -> (Output_Get:bool)
   Event OnEndCrouching()
   > Set leg_adjust_alpha(leg_adjust_alpha=1)
   > Set shouldAdjustSkirtFront(shouldAdjustSkirtFront=false)
   Event OnBeginSliding() {is=title K2Node_Event}
   > Set slideAdjustAlpha(slideAdjustAlpha:double=1) -> (Output_Get:double)
   Event OnEndSliding() {is=title K2Node_Event}
   > Set slideAdjustAlpha(slideAdjustAlpha=0)
   Event OnEnterGrindRail() {is=title K2Node_Event}
   > Set shouldAdjustSkirtFront(shouldAdjustSkirtFront=true)
   > Set leg_adjust_alpha(leg_adjust_alpha=0)
   Event OnExitGrindRail() {is=title K2Node_Event}
   > Set shouldAdjustSkirtFront(shouldAdjustSkirtFront=false)
   > Set leg_adjust_alpha(leg_adjust_alpha=1)
   n34 = OR Boolean!(A=Get bIsSkyDiving(), B=Get bIsPlayingEmote()) -> (ReturnValue)
   n35 = Branch(Condition=AND Boolean(A=Get bIsInFrontEnd(), B=NOT Boolean(A=@n34)))
   > Set Idle Animation FrontEnd(Idle Animation FrontEnd:bool=true) -> (Output_Get:bool)
   Set Idle Animation FrontEnd(execute=@n35.else, Idle Animation FrontEnd=false)
   PrintString!(execute, InString=IsReactive Set to True, bPrintToScreen, bPrintToLog, TextColor, Duration, Key) -> (then)
   PrintString(InString=IsReactiveTestFailed)
  ]}
]
variables [
 AnimDynSimSpaceSettings {
  type=AnimPhysSimSpaceSettings
  default {maxAngularAcceleration=60, maxAngularVelocity=12, simSpaceAngularAlpha=0.75, externalAngularVelocity=<0 0 0>}
 }
 BodySkelPart {type=SkeletalMeshComponent}
 CharacterPartAnimInstance {type=CustomCharacterPartAnimInstance, default=null}
 deferred_helper_alpha {type=number}
 DynamicMaterialInstance {type=MaterialInstanceDynamic}
 Gravity Override {type=Vector}
 hideMeshParts {type=boolean}
 Idle Animation FrontEnd {type=boolean}
 InvisibleThing {type=boolean}
 IsElimActive {type=boolean}
 IsFiring {type=boolean}
 IsOnZipline {type=boolean}
 isReactive {type=boolean}
 isTargeting {type=boolean}
 isTombstoneEmote {type=boolean}
 leg_adjust_alpha {type=number}
 RBAN Settings Out {type=FortRigidBodyAnimNodeParameters}
 RBAN_Alpha {type=number, default=1}
 ReactivityEnabled {type=boolean, default=true}
 shouldAdjustSkirtFront {type=boolean}
 slideAdjustAlpha {type=number}
 thigh_l_up_alpha {type=number}
 thigh_r_up_alpha {type=number}
 TombstoneEmoteActiveAlpha {type=number}
]
