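# Collision physics asset for the cloth simulation of the PolkaSkate medium female player body cosmetic, as its ClothEditor name indicates.
# Single capsules on the pelvis, two lower spine bones and both feet approximate the body, with the foot capsules mirrored left to right.
# Thigh and calf bodies exist but hold no shapes, and there are no constraints, so it defines collision volumes without ragdoll joints.
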PhysicsAsset /BRCosmetics/Characters/Player/Female/Medium/Bodies/F_MED_PolkaSkate/Meshes/ClothAsset/F_MED_PolkaSkate_ClothEditor_Collision
bodies {
 calf_l []
 calf_r []
 foot_l [{shapeType=Capsule, length=11.0765, radius=8.0344, center=<-4.1811 6.0027 -0.338>, rotation=<rot 0.0388 1.4701 -89.3982>}]
 foot_r [{shapeType=Capsule, length=11.0765, radius=8.0344, center=<4.1811 -6.0027 0.338>, rotation=<rot 0.0388 1.4701 90.6018>}]
 pelvis [{shapeType=Capsule, length=7.9943, radius=10.0538, center=<-5.1762 -1.4028 -0.1981>}]
 spine_01 [{shapeType=Capsule, length=6.272, radius=9.6812, center=<-2.9743 0.2475 0>}]
 spine_02 [{shapeType=Capsule, length=4.8309, radius=8.4035, center=<-2.6885 0.7846 0>}]
 thigh_l []
 thigh_r []
}
constraints []
